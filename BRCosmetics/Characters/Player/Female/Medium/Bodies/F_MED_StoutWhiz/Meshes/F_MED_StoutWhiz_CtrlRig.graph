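# Control Rig for the F_MED_StoutWhiz cosmetic body that corrects overalls, shorts and hanging props as the legs move.
# A sequence runs FortDeformRig for upper arms, then stores thigh spherical pose reader and remapped knee-bend alphas in variables.
# Weighted parent constraints place accessory bones, then interpolated thigh-forward and calf-bend alphas weight corrective ModifyTransforms.
# Other thigh alphas and the head reader drive nothing here.
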ControlRigBlueprint /BRCosmetics/Characters/Player/Female/Medium/Bodies/F_MED_StoutWhiz/Meshes/F_MED_StoutWhiz_CtrlRig
graphs [
 RigVMFunctionLibrary {links [], nodes []}
 RigVMFunction_Sequence_1_ContainedGraph {nodes [
   n1 = RigVMFunctionEntryNode()
   RigVMFunction_Sequence = &RigVMFunction_Sequence(ExecuteContext=@n1.ExecuteContext)
   n2 = &RigVMFunction_Sequence(ExecuteContext=@RigVMFunction_Sequence.B)
   n3 = &RigVMFunction_Sequence(ExecuteContext=@n2.B)
   n4 = &RigVMFunction_Sequence(ExecuteContext=@n3.B)
   n5 = &RigVMFunction_Sequence(ExecuteContext=@n4.B)
   n6 = &RigVMFunction_Sequence(ExecuteContext=@n5.B)
   RigVMFunctionReturnNode(A=@RigVMFunction_Sequence.A, B=@n2.A, C=@n3.A, D=@n4.A, E=@n5.A, F=@n6.A, G=@n6.B)
  ]}
 RigVMModel {nodes [
   RigUnit_BeginExecution = &RigUnit_BeginExecution()
   RigVMFunction_Sequence = RigVMAggregateNode(ExecuteContext=@RigUnit_BeginExecution.ExecutePin)
   &RigUnit_ParentConstraint(
    AdvancedSettings=(InterpolationType=Average,RotationOrderForFilter=XZY)
    Child=(Type=Bone,Name="dyn_overallsFront_l_base")
    Parents="((Item=(Type=Bone,Name=\"thigh_l\"),Weight=0.3245),(Item=(Type=Bone,Name=\"thigh_r\"),Weight=0.0232),(Item=(Type=Bone,Name=\"spine_02\"),Weight=0.0279),(Item=(Type=Bone,Name=\"spine_01\"),Weight=0.0802),(Item=(Type=Bone,Name=\"pelvis\"),Weight=0.3561),(Item=(Type=Bo ...(truncated)"
    Weight=1
    ExecutePin=@RigVMFunction_Sequence.D
   )
   > &RigUnit_ParentConstraint(
    AdvancedSettings=(InterpolationType=Average,RotationOrderForFilter=XZY)
    Child=(Type=Bone,Name="dyn_overallsFront_r_base")
    Parents="((Item=(Type=Bone,Name=\"thigh_l\"),Weight=0.0305),(Item=(Type=Bone,Name=\"thigh_r\"),Weight=0.2546),(Item=(Type=Bone,Name=\"spine_02\"),Weight=0.0367),(Item=(Type=Bone,Name=\"spine_01\"),Weight=0.0781),(Item=(Type=Bone,Name=\"pelvis\"),Weight=0.2478),(Item=(Type=Bo ...(truncated)"
    Weight=1
   )
   > &RigUnit_ParentConstraint(
    AdvancedSettings=(InterpolationType=Average,RotationOrderForFilter=XZY)
    Child=(Type=Bone,Name="dyn_compact_base")
    Parents="((Item=(Type=Bone,Name=\"thigh_l\"),Weight=0.0147),(Item=(Type=Bone,Name=\"thigh_r\"),Weight=0.1916),(Item=(Type=Bone,Name=\"spine_02\"),Weight=0.2173),(Item=(Type=Bone,Name=\"spine_01\"),Weight=0.2639),(Item=(Type=Bone,Name=\"pelvis\"),Weight=0.2389),(Item=(Type=Bo ...(truncated)"
    Weight=1
   )
   > &RigUnit_ParentConstraint(
    AdvancedSettings=(InterpolationType=Average,RotationOrderForFilter=XZY)
    Child=(Type=Bone,Name="dyn_shorts_bck_l")
    Parents=((Item=(Type=Bone,Name="calf_l"),Weight=1))
    Weight=1
   )
   > &RigUnit_ParentConstraint(
    AdvancedSettings=(InterpolationType=Average,RotationOrderForFilter=XZY)
    Child=(Type=Bone,Name="dyn_shorts_bck_r")
    Parents=((Item=(Type=Bone,Name="calf_r"),Weight=1))
    Weight=1
   )
   n1 = &RigVMFunction_AlphaInterp(
    InterpSpeedDecreasing=30
    InterpSpeedIncreasing=30
    Scale=1
    bClampResult=true
    Value=RigVMVariableNode() {variable=L thighForward Alpha}.Value
   )
   > &RigUnit_ModifyTransforms(
    ItemToModify="((Item=(Type=Bone,Name=\"dyn_overallsFront_l_base\"),Transform=(Rotation=(X=-0.0734,Y=0.1574,Z=-0.5649,W=-0.8067),Translation=(X=0.4583,Y=1.4836,Z=-0.3429),Scale3D=(X=1,Y=1,Z=1))),(Item=(Type=Bone,Name=\"dyn_gun_base\"),Transform=(Rotation=(X=0,Y=-0.1736,Z=0,W ...(truncated)"
    WeightMinimum=0
    Weight=@n1.Result
   )
   n2 = &RigVMFunction_AlphaInterp(
    InterpSpeedDecreasing=30
    InterpSpeedIncreasing=30
    Scale=1
    bClampResult=true
    Value=RigVMVariableNode() {variable=R thighForward Alpha}.Value
   )
   > &RigUnit_ModifyTransforms(
    ItemToModify="((Item=(Type=Bone,Name=\"dyn_overallsFront_r_base\"),Transform=(Rotation=(X=0,Y=-0.1677,Z=-0.2588,W=0.9513),Translation=(X=0.409,Y=-2.2247,Z=-0.156),Scale3D=(X=1,Y=1,Z=1))),(Item=(Type=Bone,Name=\"dyn_compact_base\"),Transform=(Rotation=(X=-0.4738,Y=-0.1632,Z= ...(truncated)"
    WeightMinimum=0
    Weight=@n2.Result
   )
   n3 = &RigUnit_SphericalPoseReader(
    ActiveRegionScaleFactors=(PositiveWidth=4,NegativeWidth=3,PositiveHeight=2.5,NegativeHeight=2)
    ActiveRegionSize=0.1
    Debug=(bDrawDebug=false,bDraw2D=False,bDrawLocalAxes=False,DebugScale=25,DebugSegments=20,DebugThickness=0.25)
    DriverAxis=(X=-1,Y=0,Z=0)
    DriverItem=(Type=Bone,Name="thigh_l")
    FalloffRegionScaleFactors=(PositiveWidth=1,NegativeWidth=3,PositiveHeight=1,NegativeHeight=0.5)
    FalloffSize=0.4
    FlipHeightScaling=False
    FlipWidthScaling=False
    OptionalParentItem=(Type=Bone,Name="None")
    RotationOffset=(X=100,Y=0,Z=-20)
    ExecutePin=@RigVMFunction_Sequence.B
   )
   n4 = RigVMVariableNode(Value=@n3.OutputParam, ExecuteContext=@RigVMFunction_Sequence.C) {variable=L thighForward Alpha}
   n5 = &RigUnit_SphericalPoseReader(
    ActiveRegionScaleFactors=(PositiveWidth=4,NegativeWidth=3,PositiveHeight=2.5,NegativeHeight=2)
    ActiveRegionSize=0.1
    Debug=(bDrawDebug=false,bDraw2D=False,bDrawLocalAxes=False,DebugScale=25,DebugSegments=20,DebugThickness=0.25)
    DriverAxis=(X=1,Y=0,Z=0)
    DriverItem=(Type=Bone,Name="thigh_r")
    FalloffRegionScaleFactors=(PositiveWidth=1,NegativeWidth=3.6,PositiveHeight=1,NegativeHeight=0.5)
    FalloffSize=0.4
    FlipHeightScaling=False
    FlipWidthScaling=true
    OptionalParentItem=(Type=Bone,Name="None")
    RotationOffset=(X=-80,Y=0,Z=-35.3405)
    ExecutePin=@n3.ExecutePin
   )
   n6 = RigVMVariableNode(Value=@n5.OutputParam, ExecuteContext=@n4.ExecuteContext) {variable=R thighForward Alpha}
   n7 = &RigUnit_SphericalPoseReader(
    ActiveRegionScaleFactors=(PositiveWidth=3,NegativeWidth=2,PositiveHeight=2,NegativeHeight=2)
    ActiveRegionSize=0.1
    Debug=(bDrawDebug=false,bDraw2D=False,bDrawLocalAxes=False,DebugScale=25,DebugSegments=20,DebugThickness=0.25)
    DriverAxis=(X=-1,Y=0,Z=0)
    DriverItem=(Type=Bone,Name="thigh_l")
    FalloffRegionScaleFactors=(PositiveWidth=2,NegativeWidth=2,PositiveHeight=2,NegativeHeight=2)
    FalloffSize=0.25
    FlipHeightScaling=false
    FlipWidthScaling=false
    OptionalParentItem=(Type=Bone,Name="None")
    RotationOffset=(X=180,Y=-40,Z=0)
    ExecutePin=@n5.ExecutePin
   )
   n8 = RigVMVariableNode(Value=@n7.OutputParam, ExecuteContext=@n6.ExecuteContext) {variable=L thighSide Alpha}
   n9 = &RigUnit_SphericalPoseReader(
    ActiveRegionScaleFactors=(PositiveWidth=3,NegativeWidth=2,PositiveHeight=2,NegativeHeight=2)
    ActiveRegionSize=0.1
    Debug=(bDrawDebug=false,bDraw2D=False,bDrawLocalAxes=False,DebugScale=25,DebugSegments=20,DebugThickness=0.25)
    DriverAxis=(X=1,Y=0,Z=0)
    DriverItem=(Type=Bone,Name="thigh_r")
    FalloffRegionScaleFactors=(PositiveWidth=2,NegativeWidth=2,PositiveHeight=2,NegativeHeight=2)
    FalloffSize=0.25
    FlipHeightScaling=False
    FlipWidthScaling=true
    OptionalParentItem=(Type=Bone,Name="None")
    RotationOffset=(X=0,Y=-40,Z=0)
    ExecutePin=@n7.ExecutePin
   )
   n10 = RigVMVariableNode(Value=@n9.OutputParam, ExecuteContext=@n8.ExecuteContext) {variable=R thighSide Alpha}
   n11 = &RigUnit_SphericalPoseReader(
    ActiveRegionScaleFactors=(PositiveWidth=2,NegativeWidth=2,PositiveHeight=2,NegativeHeight=1)
    ActiveRegionSize=0.1
    Debug=(bDrawDebug=false,bDraw2D=False,bDrawLocalAxes=False,DebugScale=25,DebugSegments=20,DebugThickness=0.25)
    DriverAxis=(X=-1,Y=0,Z=0)
    DriverItem=(Type=Bone,Name="thigh_l")
    FalloffRegionScaleFactors=(PositiveWidth=2,NegativeWidth=2,PositiveHeight=4,NegativeHeight=1)
    FalloffSize=0.3
    FlipHeightScaling=false
    FlipWidthScaling=false
    OptionalParentItem=(Type=Bone,Name="None")
    RotationOffset=(X=0,Y=90,Z=80)
    ExecutePin=@n9.ExecutePin
   )
   n12 = RigVMVariableNode(Value=@n11.OutputParam, ExecuteContext=@n10.ExecuteContext) {variable=L thighBack Alpha}
   n13 = &RigUnit_SphericalPoseReader(
    ActiveRegionScaleFactors=(PositiveWidth=2,NegativeWidth=2,PositiveHeight=2,NegativeHeight=1)
    ActiveRegionSize=0.1
    Debug=(bDrawDebug=false,bDraw2D=False,bDrawLocalAxes=False,DebugScale=25,DebugSegments=20,DebugThickness=0.25)
    DriverAxis=(X=1,Y=0,Z=0)
    DriverItem=(Type=Bone,Name="thigh_r")
    FalloffRegionScaleFactors=(PositiveWidth=2,NegativeWidth=2,PositiveHeight=3,NegativeHeight=1)
    FalloffSize=0.3
    FlipHeightScaling=false
    FlipWidthScaling=false
    OptionalParentItem=(Type=Bone,Name="None")
    RotationOffset=(X=180,Y=93.7803,Z=60.6503)
    ExecutePin=@n11.ExecutePin
   )
   n14 = RigVMVariableNode(Value=@n13.OutputParam, ExecuteContext=@n12.ExecuteContext) {variable=R thighBack Alpha}
   n15 = &RigUnit_SphericalPoseReader(
    ActiveRegionScaleFactors=(PositiveWidth=3,NegativeWidth=2,PositiveHeight=2,NegativeHeight=2)
    ActiveRegionSize=0.1
    Debug=(bDrawDebug=false,bDraw2D=False,bDrawLocalAxes=False,DebugScale=25,DebugSegments=20,DebugThickness=0.25)
    DriverAxis=(X=-1,Y=0,Z=0)
    DriverItem=(Type=Bone,Name="thigh_l")
    FalloffRegionScaleFactors=(PositiveWidth=2,NegativeWidth=2,PositiveHeight=2,NegativeHeight=2)
    FalloffSize=0.25
    FlipHeightScaling=false
    FlipWidthScaling=false
    OptionalParentItem=(Type=Bone,Name="None")
    RotationOffset=(X=2.688,Y=38.2586,Z=-2.2671)
    ExecutePin=@n13.ExecutePin
   )
   n16 = RigVMVariableNode(Value=@n15.OutputParam, ExecuteContext=@n14.ExecuteContext) {variable=L thighSideIn Alpha}
   n17 = &RigUnit_SphericalPoseReader(
    ActiveRegionScaleFactors=(PositiveWidth=3,NegativeWidth=2,PositiveHeight=2,NegativeHeight=2)
    ActiveRegionSize=0.1
    Debug=(bDrawDebug=false,bDraw2D=False,bDrawLocalAxes=False,DebugScale=25,DebugSegments=20,DebugThickness=0.25)
    DriverAxis=(X=1,Y=0,Z=0)
    DriverItem=(Type=Bone,Name="thigh_r")
    FalloffRegionScaleFactors=(PositiveWidth=2,NegativeWidth=2,PositiveHeight=2,NegativeHeight=2)
    FalloffSize=0.25
    FlipHeightScaling=False
    FlipWidthScaling=true
    OptionalParentItem=(Type=Bone,Name="None")
    RotationOffset=(X=0,Y=-40,Z=0)
    ExecutePin=@n15.ExecutePin
   )
   RigVMVariableNode(Value=@n17.OutputParam, ExecuteContext=@n16.ExecuteContext) {variable=R thighSideIn Alpha}
   n18 = &RigVMFunction_AlphaInterp(InterpSpeedDecreasing=30, InterpSpeedIncreasing=30, Scale=1, bClampResult=true, Value=RigVMVariableNode() {variable=L calfBend Alpha}.Value)
   &RigUnit_ModifyTransforms(
    ItemToModify="((Item=(Type=Bone,Name=\"dyn_shorts_bck_l\"),Transform=(Rotation=(X=0,Y=-0.1736,Z=0,W=0.9848),Translation=(X=3.4218,Y=0,Z=-4.0443),Scale3D=(X=1,Y=1,Z=1))),(Item=(Type=Bone,Name=\"dyn_shorts_in_l\"),Transform=(Rotation=(X=0,Y=0,Z=0.342,W=0.9397),Translation=(X= ...(truncated)"
    WeightMinimum=0
    Weight=@n18.Result
    ExecutePin=@RigVMFunction_Sequence.E
   )
   n19 = &RigVMFunction_AlphaInterp(InterpSpeedDecreasing=30, InterpSpeedIncreasing=30, Scale=1, bClampResult=true, Value=RigVMVariableNode() {variable=R calfBend Alpha}.Value)
   > &RigUnit_ModifyTransforms(
    ItemToModify="((Item=(Type=Bone,Name=\"dyn_shorts_bck_r\"),Transform=(Rotation=(X=0,Y=0,Z=0,W=1),Translation=(X=0,Y=0,Z=0),Scale3D=(X=1,Y=1,Z=1))),(Item=(Type=Bone,Name=\"dyn_shorts_in_r\"),Transform=(Rotation=(X=0.05,Y=0.0714,Z=-0.5714,W=0.816),Translation=(X=-4.5195,Y=12. ...(truncated)"
    WeightMinimum=0
    Weight=@n19.Result
   )
   /Script/FortniteGame.RigUnit_FortDeformRig(
    Bool=(bUpperArm=true,bElbow=False,bWrist=False,bGroin=False,bGlute=False,bKnee=False,bPec=False,bBicep=False,bLat=False)
    Curve="(bLUpperArmUpAlpha=False,bRUpperArmUpAlpha=False,bLUpperArmFrAlpha=False,bRUpperArmFrAlpha=False,bLUpperArmBkAlpha=False,bRUpperArmBkAlpha=False,bLWristUpAlpha=False,bRWristUpAlpha=False,bLWristDnAlpha=False,bRWristDnAlpha=False,bLKneeAlpha=False,bRKneeAlp ...(truncated)"
    Debug=(bUpperArmDebug=False,bWristDebug=False,bLatDebug=False,bGroinGluteDebug=False)
    Value="(UpperArmUp=(Rotation=(X=0,Y=0,Z=0,W=1),Translation=(X=0,Y=0,Z=0),Scale3D=(X=1,Y=1,Z=1)),UpperArmFwdFr=(Rotation=(X=0,Y=0,Z=0,W=1),Translation=(X=0,Y=0,Z=0),Scale3D=(X=1,Y=1,Z=1)),UpperArmFwdBk=(Rotation=(X=0,Y=0,Z=0,W=1),Translation=(X=0,Y=0,Z=0),Scale3D= ...(truncated)"
    ExecutePin=@RigVMFunction_Sequence.A
   ) {is=struct RigVMUnitNode}
   n20 = &RigUnit_SphericalPoseReader(
    ActiveRegionScaleFactors=(PositiveWidth=1,NegativeWidth=1,PositiveHeight=1,NegativeHeight=1)
    ActiveRegionSize=0.12
    Debug=(bDrawDebug=false,bDraw2D=False,bDrawLocalAxes=False,DebugScale=25,DebugSegments=20,DebugThickness=0.25)
    DriverAxis=(X=0,Y=1,Z=0)
    DriverItem=(Type=Bone,Name="thigh_l")
    FalloffRegionScaleFactors=(PositiveWidth=1,NegativeWidth=1,PositiveHeight=1,NegativeHeight=1)
    FalloffSize=0.17
    FlipHeightScaling=false
    FlipWidthScaling=false
    OptionalParentItem=(Type=Bone,Name="None")
    RotationOffset=(X=58.5993,Y=-9.6785,Z=-26.2186)
    ExecutePin=@n25.ExecutePin
   )
   n21 = RigVMVariableNode(ExecuteContext=(), Value=@n20.OutputParam) {variable=L thighTwistIn Alpha}
   n22 = &RigUnit_SphericalPoseReader(
    ActiveRegionScaleFactors=(PositiveWidth=1,NegativeWidth=1,PositiveHeight=1,NegativeHeight=1)
    ActiveRegionSize=0.12
    Debug=(bDrawDebug=false,bDraw2D=False,bDrawLocalAxes=False,DebugScale=25,DebugSegments=20,DebugThickness=0.25)
    DriverAxis=(X=0,Y=-1,Z=0)
    DriverItem=(Type=Bone,Name="thigh_r")
    FalloffRegionScaleFactors=(PositiveWidth=1,NegativeWidth=1,PositiveHeight=1,NegativeHeight=1)
    FalloffSize=0.17
    FlipHeightScaling=False
    FlipWidthScaling=true
    OptionalParentItem=(Type=Bone,Name="None")
    RotationOffset=(X=-127.0157,Y=4.3906,Z=-18.4663)
    ExecutePin=@n20.ExecutePin
   )
   RigVMVariableNode(Value=@n22.OutputParam, ExecuteContext=@n21.ExecuteContext) {variable=R thighTwistIn Alpha}
   n23 = &RigUnit_GetRelativeTransformForItem(Child=(Type=Bone,Name="calf_r"), Parent=(Type=Bone,Name="thigh_r"), bChildInitial=False, bParentInitial=False)
   RigVMVariableNode(
    Value=&RigVMFunction_MathDoubleRemap(
     SourceMaximum=100
     SourceMinimum=45
     TargetMaximum=1
     TargetMinimum=0
     Value=&RigVMFunction_MathQuaternionToEuler(RotationOrder=ZYX, Value=@n23.RelativeTransform.Rotation).Result.Z
    ).Result
    ExecuteContext=@n26.ExecuteContext
   ) {variable=R calfBend Alpha}
   n24 = &RigUnit_GetRelativeTransformForItem(Child=(Type=Bone,Name="calf_l"), Parent=(Type=Bone,Name="thigh_l"), bChildInitial=False, bParentInitial=False)
   RigVMFunction_MathQuaternionToEuler = &RigVMFunction_MathQuaternionToEuler(RotationOrder=ZYX, Value=@n24.RelativeTransform.Rotation)
   Remap = &RigVMFunction_MathDoubleRemap(SourceMaximum=100, SourceMinimum=45, TargetMaximum=1, TargetMinimum=0, Value=@RigVMFunction_MathQuaternionToEuler.Result.Z)
   n25 = &RigUnit_SphericalPoseReader(
    ActiveRegionScaleFactors=(PositiveWidth=1,NegativeWidth=1,PositiveHeight=1,NegativeHeight=1)
    ActiveRegionSize=0.2
    Debug=(bDrawDebug=false,bDraw2D=False,bDrawLocalAxes=False,DebugScale=25,DebugSegments=20,DebugThickness=0.25)
    DriverAxis=(X=1,Y=0,Z=0)
    DriverItem=(Type=Bone,Name="head")
    FalloffRegionScaleFactors=(PositiveWidth=1.5,NegativeWidth=1,PositiveHeight=1,NegativeHeight=1)
    FalloffSize=0.32
    FlipHeightScaling=False
    FlipWidthScaling=False
    OptionalParentItem=(Type=Bone,Name="None")
    RotationOffset=(X=-64.889,Y=15.1763,Z=36.1456)
    ExecutePin=@n17.ExecutePin
   )
   n26 = RigVMVariableNode(Value=@Remap.Result, ExecuteContext=@n22.ExecutePin) {variable=L calfBend Alpha}
  ]}
]
hierarchy [
 root {loc=<0 0 0>, rot=<0 0 0 1>}
 pelvis {parent=root, loc=<0 -0.5301 98.0571>, rot=<0.0127 0.707 -0.0127 -0.707>}
 spine_01 {parent=pelvis, loc=<6.1208 -0.9328 0>, rot=<0 0 0.034 -0.9994>}
 spine_02 {parent=spine_01, loc=<6.5997 -0.2095 0>, rot=<0 0 0.0238 -0.9997>}
 spine_03 {parent=spine_02, loc=<6.4914 -0.1237 0>, rot=<0 0 0.0149 -0.9999>}
 spine_04 {parent=spine_03, loc=<6.9598 0.0794 0>, rot=<0 0 -0.0132 -0.9999>}
 spine_05 {parent=spine_04, loc=<6.9987 0 0>, rot=<0 0 -0.0015 -1>}
 neck_01 {parent=spine_05, loc=<11.3138 0.1529 0>, rot=<0 0 -0.2003 -0.9797>}
 neck_02 {parent=neck_01, loc=<4.5627 0 0>, rot=<0 0 0.0504 -0.9987>}
 head {parent=neck_02, loc=<5.3914 0 0>, rot=<0 0 0.115 -0.9934>}
 clavicle_l {parent=spine_05, loc=<5.4831 -0.1053 -3.2946>, rot=<0.0465 -0.7289 -0.009 -0.6829>}
 upperarm_l {parent=clavicle_l, loc=<9.8658 -1.2136 -0.4105>, rot=<0.0162 -0.3173 -0.027 -0.9478>}
 lowerarm_l {parent=upperarm_l, loc=<26.465 -0.8345 0.259>, rot=<0.0036 0.0161 -0.2298 -0.9731>}
 hand_l {parent=lowerarm_l, loc=<22.7355 -0.3203 -0.6105>, rot=<-0.6784 -0.0223 0.0735 -0.7307>}
 thumb_01_l {parent=hand_l, loc=<1.3118 -0.3191 -2.459>, rot=<0.6546 -0.4042 0.0664 -0.6354>}
 thumb_02_l {parent=thumb_01_l, loc=<5.9336 -0.0448 0.0401>, rot=<-0.0037 0.0213 0.298 -0.9543>}
 thumb_03_l {parent=thumb_02_l, loc=<2.2223 0.3682 -0.0907>, rot=<0.0118 -0.013 0.0066 -0.9998>}
 index_metacarpal_l {parent=hand_l, loc=<1.6535 0.0183 -2.96>, rot=<0.0158 -0.0124 -0.0173 -0.9996>}
 index_01_l {parent=index_metacarpal_l, loc=<8.3008 0.0357 -0.0524>, rot=<-0.0044 0.0485 0.0924 -0.9945>}
 index_02_l {parent=index_01_l, loc=<3.3587 -0.3004 0.0452>, rot=<-0.002 0.0027 0.3225 -0.9466>}
 index_03_l {parent=index_02_l, loc=<2.5578 -0.123 -0.0252>, rot=<-0.002 0.0067 0.2558 -0.9667>}
 middle_metacarpal_l {parent=hand_l, loc=<2.1856 -0.1725 -0.5984>, rot=<0.006 0.0168 -0.0329 -0.9993>}
 middle_01_l {parent=middle_metacarpal_l, loc=<7.3275 -0.2447 0.0129>, rot=<-0.0208 0.0042 0.0858 -0.9961>}
 middle_02_l {parent=middle_01_l, loc=<4.0123 0.1401 0.1175>, rot=<-0.0218 0.0258 0.3512 -0.9357>}
 middle_03_l {parent=middle_02_l, loc=<2.6077 0.0534 0.1373>, rot=<0.0082 0.0188 0.1708 -0.9851>}
 ring_metacarpal_l {parent=hand_l, loc=<2.2074 -0.3625 1.3638>, rot=<-0.0284 0.0722 -0.0458 -0.9959>}
 ring_01_l {parent=ring_metacarpal_l, loc=<7.0295 -0.6197 -0.1446>, rot=<-0.0489 -0.0252 0.1892 -0.9804>}
 ring_02_l {parent=ring_01_l, loc=<3.6174 0.4647 -0.1027>, rot=<0 0 0.2806 -0.9598>}
 ring_03_l {parent=ring_02_l, loc=<2.4152 0.114 0.0199>, rot=<0 0 0.1189 -0.9929>}
 pinky_metacarpal_l {parent=hand_l, loc=<1.7616 -0.9753 3.3814>, rot=<-0.0397 0.0814 -0.069 -0.9935>}
 pinky_01_l {parent=pinky_metacarpal_l, loc=<7.1709 -0.4831 -0.1094>, rot=<-0.0654 -0.0035 0.2242 -0.9723>}
 pinky_02_l {parent=pinky_01_l, loc=<2.8274 0.0893 -0.1043>, rot=<0.0037 0.0011 0.2854 -0.9584>}
 pinky_03_l {parent=pinky_02_l, loc=<2.2046 0.0307 -0.0078>, rot=<-0.0002 0.0098 0.1445 -0.9895>}
 lowerarm_twist_02_l {parent=lowerarm_l, loc=<10.6707 0 0>, rot=<0 0 0 1>}
 lowerarm_twist_01_l {parent=lowerarm_l, loc=<18.316 0 0>, rot=<0 0 0 1>}
 upperarm_twist_01_l {parent=upperarm_l, loc=<7.2135 -0.1134 0>, rot=<0 0 0 1>}
 deform_upperarm_bk_base_l {parent=upperarm_twist_01_l, loc=<-6.4015 -4.3746 0.304>, rot=<0 -0.0061 0 -1>}
 deform_upperarm_bk_l {parent=deform_upperarm_bk_base_l, loc=<0 0 0>, rot=<0 0 0 1>}
 deform_upperarm_fr_base_l {parent=upperarm_twist_01_l, loc=<-5.6277 5.092 0.3184>, rot=<0 -0.0054 0 -1>}
 deform_upperarm_fr_l {parent=deform_upperarm_fr_base_l, loc=<0 0 0>, rot=<0 0 0 1>}
 deform_upperarm_up_base_l {parent=upperarm_twist_01_l, loc=<-8.8667 0.4825 3.9618>, rot=<0 -0.0031 0 -1>}
 deform_upperarm_up_l {parent=deform_upperarm_up_base_l, loc=<0 0 0>, rot=<0 0 0 1>}
 upperarm_twist_02_l {parent=upperarm_l, loc=<19.7998 -0.1134 0>, rot=<0 0 0 1>}
 dyn_shirtSleeve_base_l {parent=upperarm_l, loc=<15.5865 -0.8546 -1.0485>, rot=<-0.1683 0.0623 0.0332 -0.9832>}
 dyn_shirtSleeve_top_l {parent=dyn_shirtSleeve_base_l, loc=<-0.2021 1.4982 5.5913>, rot=<0.1291 0.1465 -0.0193 -0.9806>}
 dyn_shirtSleeve_fr_l {parent=dyn_shirtSleeve_base_l, loc=<-0.2768 5.4339 -0.0035>, rot=<0.6426 0.0408 -0.0284 -0.7646>}
 dyn_shirtSleeve_bot_l {parent=dyn_shirtSleeve_base_l, loc=<0.2702 -0.8151 -5.5268>, rot=<0.9926 -0.0245 -0.1105 0.0433>}
 dyn_shirtSleeve_bck_l {parent=dyn_shirtSleeve_base_l, loc=<0.4424 -4.6808 -1.2259>, rot=<-0.7659 0.0074 -0.0158 -0.6427>}
 clavicle_r {parent=spine_05, loc=<5.4831 -0.1053 3.2946>, rot=<0.7289 0.0465 -0.6829 0.009>}
 upperarm_r {parent=clavicle_r, loc=<-9.8658 1.2136 0.4105>, rot=<0.0164 -0.3245 -0.0269 -0.9454>}
 lowerarm_r {parent=upperarm_r, loc=<-26.465 0.8345 -0.259>, rot=<0.0036 0.0161 -0.2298 -0.9731>}
 hand_r {parent=lowerarm_r, loc=<-22.7355 0.3203 0.6106>, rot=<-0.6784 -0.0223 0.0735 -0.7307>}
 thumb_01_r {parent=hand_r, loc=<-1.3118 0.3191 2.459>, rot=<0.6546 -0.4042 0.0664 -0.6354>}
 thumb_02_r {parent=thumb_01_r, loc=<-5.9336 0.0448 -0.0401>, rot=<-0.0037 0.0213 0.298 -0.9543>}
 thumb_03_r {parent=thumb_02_r, loc=<-2.2223 -0.3682 0.0907>, rot=<0.0118 -0.013 0.0066 -0.9998>}
 index_metacarpal_r {parent=hand_r, loc=<-1.6535 -0.0183 2.96>, rot=<0.0158 -0.0124 -0.0173 -0.9996>}
 index_01_r {parent=index_metacarpal_r, loc=<-8.3008 -0.0357 0.0524>, rot=<-0.0044 0.0485 0.0924 -0.9945>}
 index_02_r {parent=index_01_r, loc=<-3.3587 0.3004 -0.0452>, rot=<-0.002 0.0027 0.3225 -0.9466>}
 index_03_r {parent=index_02_r, loc=<-2.5578 0.123 0.0252>, rot=<-0.002 0.0067 0.2558 -0.9667>}
 middle_metacarpal_r {parent=hand_r, loc=<-2.1856 0.1725 0.5984>, rot=<0.006 0.0168 -0.0329 -0.9993>}
 middle_01_r {parent=middle_metacarpal_r, loc=<-7.3275 0.2447 -0.0129>, rot=<-0.0208 0.0042 0.0858 -0.9961>}
 middle_02_r {parent=middle_01_r, loc=<-4.0123 -0.1401 -0.1175>, rot=<-0.0218 0.0258 0.3512 -0.9357>}
 middle_03_r {parent=middle_02_r, loc=<-2.6077 -0.0534 -0.1373>, rot=<0.0082 0.0188 0.1708 -0.9851>}
 ring_metacarpal_r {parent=hand_r, loc=<-2.2074 0.3625 -1.3638>, rot=<-0.0284 0.0722 -0.0458 -0.9959>}
 ring_01_r {parent=ring_metacarpal_r, loc=<-7.0295 0.6197 0.1446>, rot=<-0.0489 -0.0252 0.1892 -0.9804>}
 ring_02_r {parent=ring_01_r, loc=<-3.6174 -0.4647 0.1027>, rot=<0 0 0.2806 -0.9598>}
 ring_03_r {parent=ring_02_r, loc=<-2.4152 -0.114 -0.0199>, rot=<0 0 0.1189 -0.9929>}
 pinky_metacarpal_r {parent=hand_r, loc=<-1.7616 0.9753 -3.3814>, rot=<-0.0397 0.0814 -0.069 -0.9935>}
 pinky_01_r {parent=pinky_metacarpal_r, loc=<-7.1709 0.4831 0.1094>, rot=<-0.0654 -0.0035 0.2242 -0.9723>}
 pinky_02_r {parent=pinky_01_r, loc=<-2.8274 -0.0893 0.1043>, rot=<0.0037 0.0011 0.2854 -0.9584>}
 pinky_03_r {parent=pinky_02_r, loc=<-2.2046 -0.0307 0.0078>, rot=<-0.0002 0.0098 0.1445 -0.9895>}
 lowerarm_twist_02_r {parent=lowerarm_r, loc=<-10.6707 0 0>, rot=<0 0 0 -1>}
 lowerarm_twist_01_r {parent=lowerarm_r, loc=<-18.316 0 0>, rot=<0 0 0 -1>}
 upperarm_twist_01_r {parent=upperarm_r, loc=<-7.2135 0.1134 0>, rot=<0 0 0 -1>}
 deform_upperarm_bk_base_r {parent=upperarm_twist_01_r, loc=<6.4053 4.3746 -0.314>, rot=<0 0 0 -1>}
 deform_upperarm_bk_r {parent=deform_upperarm_bk_base_r, loc=<0 0 0>, rot=<0 0 0 1>}
 deform_upperarm_fr_base_r {parent=upperarm_twist_01_r, loc=<5.6312 -5.092 -0.3354>, rot=<0 0 0 -1>}
 deform_upperarm_fr_r {parent=deform_upperarm_fr_base_r, loc=<0 0 0>, rot=<0 0 0 1>}
 deform_upperarm_up_base_r {parent=upperarm_twist_01_r, loc=<8.891 -0.4825 -3.9517>, rot=<0 0 0 -1>}
 deform_upperarm_up_r {parent=deform_upperarm_up_base_r, loc=<0 0 0>, rot=<0 0 0 1>}
 upperarm_twist_02_r {parent=upperarm_r, loc=<-19.7998 0.1134 0>, rot=<0 0 0 -1>}
 dyn_shirtSleeve_base_r {parent=upperarm_r, loc=<-15.7783 0.6814 0.6702>, rot=<-0.0344 -0.9986 -0.036 -0.0151>}
 dyn_shirtSleeve_top_r {parent=dyn_shirtSleeve_base_r, loc=<-0.8823 0.2542 5.3919>, rot=<-0.0177 0.1313 -0.0065 -0.9912>}
 dyn_shirtSleeve_fr_r {parent=dyn_shirtSleeve_base_r, loc=<-0.5318 -5.0288 1.06>, rot=<-0.6861 0.0398 0.0363 -0.7256>}
 dyn_shirtSleeve_bot_r {parent=dyn_shirtSleeve_base_r, loc=<0.5574 -0.469 -5.8595>, rot=<0.9982 0.01 0.0021 0.0592>}
 dyn_shirtSleeve_bck_r {parent=dyn_shirtSleeve_base_r, loc=<0.4124 4.4033 -2.7435>, rot=<0.7998 0.0893 0.0099 -0.5934>}
 attach_backpack {parent=spine_05, loc=<-0.123 -13.473 0>, rot=<-0.0093 -0.707 0.0093 -0.707>}
 Backpack_BR {parent=attach_backpack, type=Null, loc=<0 0 0>, rot=<0 0 0 1>}
 attach_cape {parent=spine_05, loc=<9.017 -11.714 0>, rot=<0.0616 -0.7044 -0.0616 -0.7044>}
 cape {parent=attach_cape, type=Null, loc=<0 0 0>, rot=<0 0 0 1>}
 dyn_hoodieString_r_01 {parent=spine_05, loc=<6.1188 7.4059 2.1684>, rot=<0.3915 0.5486 0.6997 0.2371>}
 dyn_hoodieString_r_02 {parent=dyn_hoodieString_r_01, loc=<3.7573 0 0>, rot=<-0.0168 -0.1359 0.0792 -0.9874>}
 dyn_hoodieString_r_03 {parent=dyn_hoodieString_r_02, loc=<3.5132 0 0>, rot=<0 0 0 1>}
 dyn_hoodieString_l_01 {parent=spine_05, loc=<5.5889 7.3988 -2.2819>, rot=<-0.1941 -0.7546 -0.5589 -0.2836>}
 dyn_hoodieString_l_02 {parent=dyn_hoodieString_l_01, loc=<4.5973 0 0>, rot=<0.0534 -0.0847 -0.0443 -0.994>}
 dyn_hoodieString_l_03 {parent=dyn_hoodieString_l_02, loc=<4.5616 0 0>, rot=<0.0333 -0.0749 -0.0196 -0.9964>}
 dyn_hoodieString_l_04 {parent=dyn_hoodieString_l_03, loc=<3.5817 0 0>, rot=<0 0 0 1>}
 thigh_l {parent=pelvis, loc=<-1.0855 -0.2109 -8.7053>, rot=<0.0009 0.0355 0.0155 -0.9993>}
 calf_l {parent=thigh_l, loc=<-42.5284 -0.0019 -0.7297>, rot=<0.0006 -0.0161 -0.0391 -0.9991>}
 foot_l {parent=calf_l, loc=<-46.7401 -2.5014 -0.592>, rot=<0.0427 -0.0238 0.0353 -0.9982>}
 ball_l {parent=foot_l, loc=<-3.5658 10.8066 -0.1544>, rot=<0.0001 -0.0001 -0.7063 -0.7079>}
 calf_twist_02_l {parent=calf_l, loc=<-19.1546 -1.7943 -0.2638>, rot=<0.0237 0.0095 -0.0035 -0.9997>}
 calf_twist_01_l {parent=calf_l, loc=<-35.8483 -1.9197 -0.5794>, rot=<0.0237 0.0095 -0.0035 -0.9997>}
 thigh_twist_01_l {parent=thigh_l, loc=<-27.952 0 0>, rot=<0.0013 0 0 -1>}
 dyn_overallsFront_l_base {parent=thigh_l, loc=<-1.3323 8.4054 1.4702>, rot=<0 -0.9763 0 -0.2164>}
 dyn_overallsFront_l_01 {parent=dyn_overallsFront_l_base, loc=<0 0 0>, rot=<-0.0529 -0.0139 0.0257 -0.9982>}
 dyn_overallsFront_l_02 {parent=dyn_overallsFront_l_01, loc=<8.1473 0 0>, rot=<-0.11 0.0251 0.0608 -0.9918>}
 dyn_overallsFront_l_03 {parent=dyn_overallsFront_l_02, loc=<7.0527 0 0>, rot=<-0.0948 0.1216 0.0726 -0.9854>}
 dyn_overallsFront_l_04 {parent=dyn_overallsFront_l_03, loc=<6.5151 0 0>, rot=<0 0 0.1736 -0.9848>}
 dyn_gamePad {parent=dyn_overallsFront_l_04, loc=<0 0 0>, rot=<-0.4535 -0.5247 0.5382 0.4789>}
 dyn_overallStrap_c_l {parent=thigh_l, loc=<-21.9439 -0.9994 -10.2366>, rot=<-0.0023 -0.9369 0.0007 -0.3497>}
 dyn_shorts_fr_l {parent=thigh_l, loc=<-37.423 6.5716 -1.2634>, rot=<-0.0391 -0.7081 -0.7046 -0.0259>}
 dyn_shorts_in_l {parent=thigh_l, loc=<-36.8658 -1.5877 6.8167>, rot=<0.0877 -0.0004 0.9961 0.0024>}
 dyn_shorts_bck_l {parent=thigh_l, loc=<-37.331 -7.2128 -0.0604>, rot=<0.0663 -0.7085 0.7006 -0.0533>}
 dyn_shorts_out_l {parent=thigh_l, loc=<-37.7926 -1.0063 -6.2591>, rot=<-0.0024 -0.9987 -0.0001 -0.0518>}
 thigh_r {parent=pelvis, loc=<-1.0855 -0.2109 8.7053>, rot=<0.0355 -0.0009 0.9993 0.0155>}
 calf_r {parent=thigh_r, loc=<42.5284 0.0019 0.7297>, rot=<0.0006 -0.0161 -0.0391 -0.9991>}
 foot_r {parent=calf_r, loc=<46.7401 2.5014 0.592>, rot=<0.0427 -0.0238 0.0353 -0.9982>}
 ball_r {parent=foot_r, loc=<3.5658 -10.8066 0.1544>, rot=<0.0001 -0.0001 -0.7063 -0.7079>}
 calf_twist_02_r {parent=calf_r, loc=<19.1546 1.7943 0.2638>, rot=<0.0237 0.0095 -0.0035 -0.9997>}
 calf_twist_01_r {parent=calf_r, loc=<35.8483 1.9197 0.5794>, rot=<0.0237 0.0095 -0.0035 -0.9997>}
 thigh_twist_01_r {parent=thigh_r, loc=<27.952 0 0>, rot=<0.0013 0 0 -1>}
 dyn_overallsFront_r_base {parent=thigh_r, loc=<1.3323 -8.4054 -1.4704>, rot=<-0.0846 0.2237 -0.0172 -0.9708>}
 dyn_overallsFront_r_01 {parent=dyn_overallsFront_r_base, loc=<0 0 0>, rot=<0.0865 -0.0256 -0.0251 -0.9956>}
 dyn_overallsFront_r_02 {parent=dyn_overallsFront_r_01, loc=<8.1474 0 0>, rot=<0.1443 0.0505 -0.0819 -0.9848>}
 dyn_overallsFront_r_03 {parent=dyn_overallsFront_r_02, loc=<7.9975 0 0>, rot=<0.0613 0.0561 -0.0634 -0.9945>}
 dyn_overallsFront_r_04 {parent=dyn_overallsFront_r_03, loc=<5.4778 0 0>, rot=<0 0 -0.1305 -0.9914>}
 dyn_banana_base {parent=dyn_overallsFront_r_03, loc=<0 0 0>, rot=<-0.2988 0.2988 0.6409 -0.6409>}
 dyn_banana_body {parent=dyn_banana_base, loc=<0 0 0>, rot=<0.0114 -0.015 -0.0154 -0.9997>}
 dyn_banana_arm_r {parent=dyn_banana_body, loc=<4.1161 5.0016 -3.6645>, rot=<0.2794 -0.2697 -0.5362 -0.7495>}
 dyn_banana_arm_l {parent=dyn_banana_body, loc=<3.4189 -4.1031 -2.6956>, rot=<-0.2826 -0.1297 0.6148 -0.7248>}
 dyn_banana_lowerBody {parent=dyn_banana_body, loc=<5.2393 0.1689 -5.4828>, rot=<0.0427 -0.5317 0.0945 -0.8405>}
 dyn_banana_leg_l {parent=dyn_banana_lowerBody, loc=<4.7833 -2.4831 -0.9906>, rot=<0.0259 -0.0141 0.1451 -0.989>}
 dyn_banana_leg_r {parent=dyn_banana_lowerBody, loc=<4.3145 3.8462 -0.791>, rot=<0.0719 -0.0023 -0.1812 -0.9808>}
 dyn_banana_wire_01 {parent=dyn_banana_body, loc=<2.2602 1.1181 8.7245>, rot=<0.7375 0.5308 0.2543 -0.3312>}
 dyn_banana_wire_02 {parent=dyn_banana_wire_01, loc=<7.9765 0 0>, rot=<0 0 0 1>}
 dyn_overallStrap_c_r {parent=thigh_r, loc=<21.9439 0.9994 10.237>, rot=<0.0007 0.3497 0.0023 -0.9369>}
 dyn_shorts_fr_r {parent=thigh_r, loc=<37.423 -6.5717 1.2635>, rot=<-0.643 -0.0088 0.0046 -0.7658>}
 dyn_shorts_in_r {parent=thigh_r, loc=<36.8658 1.5877 -6.8163>, rot=<0.9969 -0.0024 -0.079 -0.0004>}
 dyn_shorts_bck_r {parent=thigh_r, loc=<37.3435 7.2126 0.2371>, rot=<0.7006 0.0533 -0.0663 -0.7085>}
 dyn_shorts_out_r {parent=thigh_r, loc=<37.7926 1.0062 6.2588>, rot=<-0.0177 0.0081 0.0008 -0.9998>}
 dyn_compact_base {parent=pelvis, loc=<9.7776 2.9245 10.4313>, rot=<0.2826 0.4861 0.8258 -0.0431>}
 dyn_compact {parent=dyn_compact_base, loc=<0 0 0>, rot=<0 0 0 1>}
 dyn_compactBot {parent=dyn_compact, loc=<5.4976 0.053 -0.5795>, rot=<0.0445 0.2159 0.0009 -0.9754>}
 dyn_gun_base {parent=pelvis, loc=<1.3231 1.6589 -13.0934>, rot=<0.1891 -0.6446 -0.2819 -0.685>}
 dyn_gun {parent=dyn_gun_base, loc=<0 0 0>, rot=<0 0 0 1>}
 dyn_overallsBack_base {parent=pelvis, loc=<5.3339 -10.9655 0>, rot=<0.5 -0.5 0.5 -0.5>}
 dyn_overallsBack_mainFlap {parent=dyn_overallsBack_base, loc=<0 0 0>, rot=<-0.0289 -0.5529 0.04 -0.8318>}
 dyn_overallsBack_r_01 {parent=dyn_overallsBack_mainFlap, loc=<5.0281 -4.2003 -0.1835>, rot=<-0.3209 -0.0187 0.21 -0.9234>}
 dyn_overallsBack_r_02 {parent=dyn_overallsBack_r_01, loc=<5.1104 0 0>, rot=<-0.3075 0.0578 0.1801 -0.9326>}
 dyn_overallsBack_r_03 {parent=dyn_overallsBack_r_02, loc=<6.3947 0 0>, rot=<-0.167 0.0123 0.065 -0.9837>}
 dyn_overallsBack_r_04 {parent=dyn_overallsBack_r_03, loc=<7.6227 0 0>, rot=<-0.0259 0.0385 0.1053 -0.9934>}
 dyn_overallsBack_r_05 {parent=dyn_overallsBack_r_04, loc=<7.0341 0 0>, rot=<0 -0.0227 0.1285 -0.9914>}
 dyn_overallsBack_l_01 {parent=dyn_overallsBack_mainFlap, loc=<4.7273 4.4317 -0.9841>, rot=<0.3664 -0.0372 -0.2379 -0.8988>}
 dyn_overallsBack_l_02 {parent=dyn_overallsBack_l_01, loc=<5.1105 0 0>, rot=<0.3479 0.0498 -0.1825 -0.9183>}
 dyn_overallsBack_l_03 {parent=dyn_overallsBack_l_02, loc=<6.3944 0 0>, rot=<0.1669 0.0124 -0.065 -0.9837>}
 dyn_overallsBack_l_04 {parent=dyn_overallsBack_l_03, loc=<7.6232 0 0>, rot=<0.1124 0.0475 -0.1016 -0.9873>}
 dyn_overallsBack_l_05 {parent=dyn_overallsBack_l_04, loc=<7.0341 0 0>, rot=<0 0 -0.1305 -0.9914>}
]
variables [
 L calfBend Alpha {type=float, default=0}
 L thighBack Alpha {type=double, default=0}
 L thighForward Alpha {type=double, default=0}
 L thighSide Alpha {type=double, default=0}
 L thighSideIn Alpha {type=double, default=0}
 L thighTwistIn Alpha {type=double, default=0}
 R calfBend Alpha {type=float, default=0}
 R thighBack Alpha {type=double, default=0}
 R thighForward Alpha {type=double, default=0}
 R thighSide Alpha {type=double, default=0}
 R thighSideIn Alpha {type=double, default=0}
 R thighTwistIn Alpha {type=double, default=0}
]
