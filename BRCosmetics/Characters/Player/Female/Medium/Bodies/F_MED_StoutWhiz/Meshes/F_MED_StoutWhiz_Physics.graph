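# Physics asset for the F_MED_StoutWhiz player body cosmetic, rigging capsule bodies for dangling overalls, a banana and a gamepad.
# Back chains branch from a pelvis-anchored flap and front chains hang from each thigh, mostly with free angular motion.
# Center strap bodies close each side into a loop through constraints duplicated in both directions.
# The banana hangs off the right front chain, the gamepad off the left, alongside limb and torso capsules.
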PhysicsAsset /BRCosmetics/Characters/Player/Female/Medium/Bodies/F_MED_StoutWhiz/Meshes/F_MED_StoutWhiz_Physics
bodies {
 calf_l [{shapeType=Capsule, length=32.7847, radius=9.7339, shapeName=calf_l_capsule, center=<-6.3117 0 1.3563>, rotation=<rot -90 -0.0002 0>}]
 calf_r [{shapeType=Capsule, length=32.9274, radius=9.9271, shapeName=calf_r_capsule, center=<6.3117 0 -2.1794>, rotation=<rot 90 -0.0002 0>}]
 dyn_banana_base [{shapeType=Capsule, length=0, radius=3, shapeName=dyn_banana_base_capsule, center=<2.7424 0 -2.5764>, rotation=<rot -0.0001 -179.9998 180>}]
 dyn_banana_body [{
   shapeType=Capsule
   length=21
   radius=7.5
   shapeName=dyn_banana_body_capsule
   center=<8.3184 0 -5.4854>
   rotation=<rot -10.0001 -179.9998 179.9999>
  }]
 dyn_gamePad [{shapeType=Capsule, length=5, radius=6, shapeName=dyn_gamePad_capsule, center=<4 -0.0889 0>, rotation=<rot 0.0001 -10.0001 89.9998>}]
 dyn_overallStrap_c_l [{shapeType=Capsule, length=2, radius=2.5, shapeName=dyn_overallStrap_c_l_capsule, center=<0.3552 0 0.7783>, rotation=<rot 0 0.0001 -90.0001>}]
 dyn_overallStrap_c_r [{shapeType=Capsule, length=2, radius=2.5, shapeName=dyn_overallStrap_c_r_capsule, center=<0.1736 0 0.3239>, rotation=<rot 0 -0.0001 -89.9999>}]
 dyn_overallsBack_base [{shapeType=Capsule, length=0, radius=1.5, shapeName=dyn_overallsBack_base_capsule, center=<2.2137 0 0>, rotation=<rot -90 0 0>}]
 dyn_overallsBack_l_01 [{shapeType=Capsule, length=5, radius=2.5, shapeName=dyn_overallsBack_l_01_capsule, center=<2.2137 0 0>, rotation=<rot -90 0.0001 0>}]
 dyn_overallsBack_l_02 [{shapeType=Capsule, length=3, radius=2.5, shapeName=dyn_overallsBack_l_02_capsule, center=<2.2137 0 0>, rotation=<rot -90 0.0001 0>}]
 dyn_overallsBack_l_03 [{shapeType=Capsule, length=3, radius=2.5, shapeName=dyn_overallsBack_l_03_capsule, center=<2.2137 -0.6878 0>, rotation=<rot -90 0 0>}]
 dyn_overallsBack_l_04 [{shapeType=Capsule, length=3, radius=2.5, shapeName=dyn_overallsBack_l_04_capsule, center=<2.2137 -1.3928 0>, rotation=<rot -90 0 0>}]
 dyn_overallsBack_l_05 [{shapeType=Capsule, length=2, radius=2.5, shapeName=dyn_overallsBack_l_05_capsule, center=<1.6693 -1.0776 0>, rotation=<rot -90 0.0001 0>}]
 dyn_overallsBack_mainFlap [{shapeType=Capsule, length=5.5, radius=3.5, shapeName=dyn_overallsBack_mainFlap_capsule, center=<4.9744 0 0>, rotation=<rot -90 0 0>}]
 dyn_overallsBack_r_01 [{shapeType=Capsule, length=3, radius=2.5, shapeName=dyn_overallsBack_r_01_capsule, center=<2.2137 0 0>, rotation=<rot -90 -0.0001 0>}]
 dyn_overallsBack_r_02 [{shapeType=Capsule, length=3, radius=2.5, shapeName=dyn_overallsBack_r_02_capsule, center=<2.2137 0.2611 0>, rotation=<rot -90 -0.0001 0>}]
 dyn_overallsBack_r_03 [{shapeType=Capsule, length=3, radius=2.5, shapeName=dyn_overallsBack_r_03_capsule, center=<2.2137 0.2611 0>, rotation=<rot -90 0 0>}]
 dyn_overallsBack_r_04 [{shapeType=Capsule, length=3, radius=2.5, shapeName=dyn_overallsBack_r_04_capsule, center=<2.2137 0.2611 0>, rotation=<rot -90 0 0>}]
 dyn_overallsBack_r_05 [{shapeType=Capsule, length=3, radius=2.5, shapeName=dyn_overallsBack_r_05_capsule, center=<2.2137 0.2611 0>, rotation=<rot -90 -0.0001 0>}]
 dyn_overallsFront_l_01 [{shapeType=Capsule, length=3, radius=2.5, shapeName=dyn_overallsFront_l_01_capsule, center=<2.2136 2.4 0>, rotation=<rot -90 0.0001 0>}]
 dyn_overallsFront_l_02 [{shapeType=Capsule, length=3, radius=2.5, shapeName=dyn_overallsFront_l_02_capsule, center=<2.2137 1.2401 0>, rotation=<rot -90 0.0001 0>}]
 dyn_overallsFront_l_03 [{shapeType=Capsule, length=2.5, radius=3, shapeName=dyn_overallsFront_l_03_capsule, center=<2.2137 0.9987 0>, rotation=<rot -90 0 0>}]
 dyn_overallsFront_l_04 [{shapeType=Capsule, length=3.5, radius=1, shapeName=dyn_overallsFront_l_04_capsule, center=<1.4511 1.1682 0.916>, rotation=<rot -90 0.0001 0>}]
 dyn_overallsFront_l_base [{shapeType=Capsule, length=2, radius=1, shapeName=dyn_overallsFront_l_base_capsule}]
 dyn_overallsFront_r_01 [{shapeType=Capsule, length=3, radius=2.5, shapeName=dyn_overallsFront_r_01_capsule, center=<2.8707 -2.4433 0>, rotation=<rot -90 -0.0001 0>}]
 dyn_overallsFront_r_02 [{shapeType=Capsule, length=3.5, radius=2.5, shapeName=dyn_overallsFront_r_02_capsule, center=<2.2137 -2.0053 0>, rotation=<rot -90 -0.0001 0>}]
 dyn_overallsFront_r_03 [{shapeType=Capsule, length=3, radius=2.5, shapeName=dyn_overallsFront_r_03_capsule, center=<1.6961 -1.6897 0>, rotation=<rot -90 0 0>}]
 dyn_overallsFront_r_04 [{shapeType=Capsule, length=3, radius=2.5, shapeName=dyn_overallsFront_r_04_capsule, center=<2.2137 -1.4428 0>, rotation=<rot -90 -0.0001 0>}]
 dyn_overallsFront_r_base [{shapeType=Capsule, length=2, radius=1, shapeName=dyn_overallsFront_r_base_capsule}]
 lowerarm_l [{shapeType=Capsule, length=20.5, radius=6.5, shapeName=lowerarm_l_capsule, center=<9.6957 0 0>, rotation=<rot -90 0 0>}]
 lowerarm_r [{shapeType=Capsule, length=12, radius=6.5, shapeName=lowerarm_r_capsule, center=<-9.6957 -0.0001 0>, rotation=<rot 90 0.0006 0>}]
 pelvis [{shapeType=Capsule, length=15.0021, radius=9.5584, shapeName=pelvis_capsule, center=<-1.9146 -0.8206 0>}]
 spine_01 [{shapeType=Capsule, length=11.5538, radius=7.2496, shapeName=spine_01_capsule}]
 spine_02 [{shapeType=Capsule, length=8.0402, radius=7.9102, shapeName=spine_02_capsule}]
 thigh_l [{shapeType=Capsule, length=30, radius=10, shapeName=thigh_l_capsule, center=<-19.5884 -1.51 2.6>, rotation=<rot 90 -0.0002 0>}]
 thigh_r [{shapeType=Capsule, length=30, radius=10, shapeName=thigh_r_capsule, center=<19.2087 1.5127 -2.6187>, rotation=<rot -90 -0.0002 0>}]
 upperarm_l [{shapeType=Capsule, length=16, radius=8, shapeName=upperarm_l_capsule, center=<15.9061 0 -0.6211>, rotation=<rot -90 0 0>}]
 upperarm_r [{shapeType=Capsule, length=16, radius=8, shapeName=upperarm_r_capsule, center=<-15.9061 -0.0002 0.6211>, rotation=<rot 90 0.0009 0>}]
}
constraints [
 {bone1Name=dyn_overallsBack_base, bone2Name=pelvis, swing1LimitDegrees=45, swing2LimitDegrees=45, twistLimitDegrees=1}
 {bone1Name=dyn_overallsBack_mainFlap, bone2Name=dyn_overallsBack_base, swing1LimitDegrees=45, swing2LimitDegrees=45, twistLimitDegrees=1}
 {
  bone1Name=dyn_overallsBack_l_01
  bone2Name=dyn_overallsBack_mainFlap
  swing1LimitDegrees=45
  swing1Motion=Free
  swing2LimitDegrees=45
  swing2Motion=Free
  twistLimitDegrees=1
  twistMotion=Free
 }
 {
  bone1Name=dyn_overallsBack_l_02
  bone2Name=dyn_overallsBack_l_01
  swing1LimitDegrees=45
  swing1Motion=Free
  swing2LimitDegrees=45
  swing2Motion=Free
  twistLimitDegrees=1
  twistMotion=Free
 }
 {
  bone1Name=dyn_overallsBack_l_03
  bone2Name=dyn_overallsBack_l_02
  swing1LimitDegrees=45
  swing1Motion=Free
  swing2LimitDegrees=45
  swing2Motion=Free
  twistLimitDegrees=1
  twistMotion=Free
 }
 {
  bone1Name=dyn_overallsBack_l_04
  bone2Name=dyn_overallsBack_l_03
  swing1LimitDegrees=45
  swing1Motion=Free
  swing2LimitDegrees=45
  swing2Motion=Free
  twistLimitDegrees=1
  twistMotion=Free
 }
 {
  bone1Name=dyn_overallsBack_l_05
  bone2Name=dyn_overallsBack_l_04
  swing1LimitDegrees=45
  swing1Motion=Free
  swing2LimitDegrees=45
  swing2Motion=Free
  twistLimitDegrees=1
  twistMotion=Free
 }
 {
  bone1Name=dyn_overallsBack_r_01
  bone2Name=dyn_overallsBack_mainFlap
  swing1LimitDegrees=45
  swing1Motion=Free
  swing2LimitDegrees=45
  swing2Motion=Free
  twistLimitDegrees=1
  twistMotion=Free
 }
 {
  bone1Name=dyn_overallsBack_r_02
  bone2Name=dyn_overallsBack_r_01
  swing1LimitDegrees=45
  swing1Motion=Free
  swing2LimitDegrees=45
  swing2Motion=Free
  twistLimitDegrees=1
  twistMotion=Free
 }
 {
  bone1Name=dyn_overallsBack_r_03
  bone2Name=dyn_overallsBack_r_02
  swing1LimitDegrees=45
  swing1Motion=Free
  swing2LimitDegrees=45
  swing2Motion=Free
  twistLimitDegrees=1
  twistMotion=Free
 }
 {
  bone1Name=dyn_overallsBack_r_04
  bone2Name=dyn_overallsBack_r_03
  swing1LimitDegrees=45
  swing1Motion=Free
  swing2LimitDegrees=45
  swing2Motion=Free
  twistLimitDegrees=1
  twistMotion=Free
 }
 {
  bone1Name=dyn_overallsBack_r_05
  bone2Name=dyn_overallsBack_r_04
  swing1LimitDegrees=45
  swing1Motion=Free
  swing2LimitDegrees=45
  swing2Motion=Free
  twistLimitDegrees=1
  twistMotion=Free
 }
 {bone1Name=dyn_overallsFront_l_base, bone2Name=thigh_l, swing1LimitDegrees=45, swing2LimitDegrees=45, twistLimitDegrees=1}
 {
  bone1Name=dyn_overallsFront_l_01
  bone2Name=dyn_overallsFront_l_base
  swing1LimitDegrees=45
  swing1Motion=Free
  swing2LimitDegrees=45
  swing2Motion=Free
  twistLimitDegrees=1
  twistMotion=Free
 }
 {
  bone1Name=dyn_overallsFront_l_02
  bone2Name=dyn_overallsFront_l_01
  swing1LimitDegrees=45
  swing1Motion=Free
  swing2LimitDegrees=45
  swing2Motion=Free
  twistLimitDegrees=1
  twistMotion=Free
 }
 {
  bone1Name=dyn_overallsFront_l_03
  bone2Name=dyn_overallsFront_l_02
  swing1LimitDegrees=45
  swing1Motion=Free
  swing2LimitDegrees=45
  swing2Motion=Free
  twistLimitDegrees=1
  twistMotion=Free
 }
 {
  bone1Name=dyn_overallsFront_l_04
  bone2Name=dyn_overallsFront_l_03
  swing1LimitDegrees=45
  swing1Motion=Free
  swing2LimitDegrees=45
  swing2Motion=Free
  twistLimitDegrees=1
  twistMotion=Free
 }
 {bone1Name=dyn_overallsFront_r_base, bone2Name=thigh_r, swing1LimitDegrees=45, swing2LimitDegrees=45, twistLimitDegrees=1}
 {
  bone1Name=dyn_overallsFront_r_01
  bone2Name=dyn_overallsFront_r_base
  swing1LimitDegrees=45
  swing1Motion=Free
  swing2LimitDegrees=45
  swing2Motion=Free
  twistLimitDegrees=1
  twistMotion=Free
 }
 {
  bone1Name=dyn_overallsFront_r_02
  bone2Name=dyn_overallsFront_r_01
  swing1LimitDegrees=45
  swing1Motion=Free
  swing2LimitDegrees=45
  swing2Motion=Free
  twistLimitDegrees=1
  twistMotion=Free
 }
 {
  bone1Name=dyn_overallsFront_r_03
  bone2Name=dyn_overallsFront_r_02
  swing1LimitDegrees=45
  swing1Motion=Free
  swing2LimitDegrees=45
  swing2Motion=Free
  twistLimitDegrees=1
  twistMotion=Free
 }
 {
  bone1Name=dyn_banana_base
  bone2Name=dyn_overallsFront_r_03
  swing1LimitDegrees=45
  swing1Motion=Free
  swing2LimitDegrees=45
  swing2Motion=Free
  twistLimitDegrees=1
  twistMotion=Free
 }
 {bone1Name=dyn_banana_body, bone2Name=dyn_banana_base, swing1LimitDegrees=45, swing2LimitDegrees=45, twistLimitDegrees=5}
 {
  bone1Name=thigh_l
  bone2Name=pelvis
  swing1LimitDegrees=45
  swing1Motion=Free
  swing2LimitDegrees=45
  swing2Motion=Free
  twistLimitDegrees=45
  twistMotion=Free
 }
 {
  bone1Name=thigh_r
  bone2Name=pelvis
  swing1LimitDegrees=45
  swing1Motion=Free
  swing2LimitDegrees=45
  swing2Motion=Free
  twistLimitDegrees=45
  twistMotion=Free
 }
 {
  bone1Name=dyn_overallsFront_r_04
  bone2Name=dyn_overallsFront_r_03
  swing1LimitDegrees=45
  swing1Motion=Free
  swing2LimitDegrees=45
  swing2Motion=Free
  twistLimitDegrees=1
  twistMotion=Free
 }
 {
  bone1Name=dyn_overallStrap_c_l
  bone2Name=dyn_overallsFront_l_04
  swing1LimitDegrees=45
  swing1Motion=Free
  swing2LimitDegrees=45
  swing2Motion=Free
  twistLimitDegrees=45
  twistMotion=Free
 }
 {
  bone1Name=dyn_overallsFront_l_04
  bone2Name=dyn_overallStrap_c_l
  swing1LimitDegrees=45
  swing1Motion=Free
  swing2LimitDegrees=45
  swing2Motion=Free
  twistLimitDegrees=45
  twistMotion=Free
 }
 {
  bone1Name=dyn_overallStrap_c_l
  bone2Name=dyn_overallsBack_l_05
  swing1LimitDegrees=45
  swing1Motion=Free
  swing2LimitDegrees=45
  swing2Motion=Free
  twistLimitDegrees=45
  twistMotion=Free
 }
 {
  bone1Name=dyn_overallsBack_l_05
  bone2Name=dyn_overallStrap_c_l
  swing1LimitDegrees=45
  swing1Motion=Free
  swing2LimitDegrees=45
  swing2Motion=Free
  twistLimitDegrees=45
  twistMotion=Free
 }
 {
  bone1Name=dyn_overallStrap_c_r
  bone2Name=dyn_overallsBack_r_05
  swing1LimitDegrees=45
  swing1Motion=Free
  swing2LimitDegrees=45
  swing2Motion=Free
  twistLimitDegrees=45
  twistMotion=Free
 }
 {
  bone1Name=dyn_overallsBack_r_05
  bone2Name=dyn_overallStrap_c_r
  swing1LimitDegrees=45
  swing1Motion=Free
  swing2LimitDegrees=45
  swing2Motion=Free
  twistLimitDegrees=45
  twistMotion=Free
 }
 {
  bone1Name=dyn_overallStrap_c_r
  bone2Name=dyn_overallsFront_r_04
  swing1LimitDegrees=45
  swing1Motion=Free
  swing2LimitDegrees=45
  swing2Motion=Free
  twistLimitDegrees=45
  twistMotion=Free
 }
 {
  bone1Name=dyn_overallsFront_r_04
  bone2Name=dyn_overallStrap_c_r
  swing1LimitDegrees=45
  swing1Motion=Free
  swing2LimitDegrees=45
  swing2Motion=Free
  twistLimitDegrees=45
  twistMotion=Free
 }
 {bone1Name=dyn_gamePad, bone2Name=dyn_overallsFront_l_04, swing1LimitDegrees=80, swing2LimitDegrees=80, twistLimitDegrees=10}
]
