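# Animation Blueprint for the StoutWhiz female body cosmetic, adding secondary motion to clothing and dangling accessories.
# It copies the parent mesh pose through a Control Rig, then chains RigidBody simulation, bone constraints for overall straps and banana, and spring-based AnimDynamics for compact, banana, shorts, sleeves and hoodie strings.
# Events adjust clothing spring stiffness for skydiving and sim-space and alpha variables for front end versus gameplay.
# Some sleeve and compact nodes sit disconnected.
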AnimBlueprint /BRCosmetics/Characters/Player/Female/Medium/Bodies/F_MED_StoutWhiz/Meshes/F_MED_StoutWhiz_AnimBP
target_skeleton=&/F_MED_StoutWhiz_Skeleton
graphs [
 AnimGraph {nodes [
   n1 = Anim Dynamics - Bone: dyn_shirtSleeve_fr_l(Alpha=1, ExternalForce="0,0,0") {
    angularDampingOverride=0.9
    angularSpringConstant=5000
    bAngularSpring=true
    bLinearSpring=true
    bOverrideAngularDamping=true
    bOverrideLinearDamping=true
    bUseGravityOverride=true
    linearDampingOverride=1
    linearSpringConstant=40
    simulationSpace=RootRelative
    alphaBoolBlend {blendOption=Linear}
    boundBone {boneName=dyn_shirtSleeve_fr_l}
    componentAppliedLinearAccClamp=<1000 1000 1000>
    componentLinearAccScale=<0.5 0.5 0.5>
    componentLinearVelScale=<0.5 0.5 0.5>
    gravityOverride {z=-500}
    physicsBodyDefinitions [{
      boundBone {boneName=dyn_shirtSleeve_fr_l}
      boxExtents=<14 14 14>
      constraintSetup {angularLimitsMax=<0 10 0>, angularLimitsMin=<0 -30 0>, angularTarget=<0 0 0>, linearAxesMax=<1 0.5 0>, linearAxesMin=<-2.5 -0.5 0>}
      localJointOffset=<0.5 0 0>
     }]
   }
   n2 = Anim Dynamics - Bone: dyn_shirtSleeve_top_l(ComponentPose=@n1, Alpha=1, ExternalForce="0,0,0") {
    angularDampingOverride=0.9
    angularSpringConstant=5000
    bAngularSpring=true
    bLinearSpring=true
    bOverrideAngularDamping=true
    bOverrideLinearDamping=true
    bUseGravityOverride=true
    linearDampingOverride=1
    linearSpringConstant=40
    simulationSpace=RootRelative
    alphaBoolBlend {blendOption=Linear}
    boundBone {boneName=dyn_shirtSleeve_top_l}
    componentAppliedLinearAccClamp=<1000 1000 1000>
    componentLinearAccScale=<0.5 0.5 0.5>
    componentLinearVelScale=<0.5 0.5 0.5>
    gravityOverride {z=-500}
    physicsBodyDefinitions [{
      boundBone {boneName=dyn_shirtSleeve_top_l}
      boxExtents=<14 14 14>
      constraintSetup {angularLimitsMax=<0 2 10>, angularLimitsMin=<0 -30 -10>, angularTarget=<0 0 0>, linearAxesMax=<1 0.5 0>, linearAxesMin=<-2.5 -0.5 0>}
      localJointOffset=<0.5 0 0>
     }]
   }
   n3 = Anim Dynamics - Bone: dyn_shirtSleeve_bot_l(ComponentPose=@n2, Alpha=1, ExternalForce="0,0,0") {
    angularDampingOverride=0.9
    angularSpringConstant=5000
    bAngularSpring=true
    bLinearSpring=true
    bOverrideAngularDamping=true
    bOverrideLinearDamping=true
    bUseGravityOverride=true
    linearDampingOverride=1
    linearSpringConstant=40
    simulationSpace=RootRelative
    alphaBoolBlend {blendOption=Linear}
    boundBone {boneName=dyn_shirtSleeve_bot_l}
    componentAppliedLinearAccClamp=<1000 1000 1000>
    componentLinearAccScale=<0.5 0.5 0.5>
    componentLinearVelScale=<0.5 0.5 0.5>
    gravityOverride {z=-500}
    physicsBodyDefinitions [{
      boundBone {boneName=dyn_shirtSleeve_bot_l}
      boxExtents=<14 14 14>
      constraintSetup {angularLimitsMax=<0 2 5>, angularLimitsMin=<0 -30 -5>, angularTarget=<0 0 0>, linearAxesMax=<1 0.5 0>, linearAxesMin=<-2.5 -0.5 0>}
      localJointOffset=<0.5 0 0>
     }]
   }
   Anim Dynamics - Bone: dyn_shirtSleeve_bck_l(ComponentPose=@n3, Alpha=1, ExternalForce="0,0,0") {
    angularDampingOverride=0.9
    angularSpringConstant=5000
    bAngularSpring=true
    bLinearSpring=true
    bOverrideAngularDamping=true
    bOverrideLinearDamping=true
    bUseGravityOverride=true
    linearDampingOverride=1
    linearSpringConstant=40
    simulationSpace=RootRelative
    alphaBoolBlend {blendOption=Linear}
    boundBone {boneName=dyn_shirtSleeve_bck_l}
    componentAppliedLinearAccClamp=<1000 1000 1000>
    componentLinearAccScale=<0.5 0.5 0.5>
    componentLinearVelScale=<0.5 0.5 0.5>
    gravityOverride {z=-500}
    physicsBodyDefinitions [{
      boundBone {boneName=dyn_shirtSleeve_bck_l}
      boxExtents=<14 14 14>
      constraintSetup {angularLimitsMax=<0 2 5>, angularLimitsMin=<0 -30 -5>, angularTarget=<0 0 0>, linearAxesMax=<1 0.5 0>, linearAxesMin=<-2.5 -0.5 0>}
      localJointOffset=<0.5 0 0>
     }]
   }
   Anim Dynamics - Bone: dyn_shirtSleeve_fr_r(Alpha=1, ExternalForce="0,0,0") {
    angularDampingOverride=0.9
    angularSpringConstant=5000
    bAngularSpring=true
    bGravityOverrideInSimSpace=true
    bLinearSpring=true
    bOverrideAngularDamping=true
    bOverrideLinearDamping=true
    bUseGravityOverride=true
    linearDampingOverride=1
    linearSpringConstant=40
    simulationSpace=RootRelative
    alphaBoolBlend {blendOption=Linear}
    boundBone {boneName=dyn_shirtSleeve_fr_r}
    componentAppliedLinearAccClamp=<1000 1000 1000>
    componentLinearAccScale=<0.5 0.5 0.5>
    componentLinearVelScale=<0.5 0.5 0.5>
    gravityOverride {z=-980}
    physicsBodyDefinitions [{
      boundBone {boneName=dyn_shirtSleeve_fr_r}
      boxExtents=<12 12 12>
      constraintSetup {angularLimitsMax=<0 10 0>, angularLimitsMin=<0 -30 0>, angularTarget=<0 0 0>, linearAxesMax=<1 0.5 0>, linearAxesMin=<-2.5 -0.5 0>}
      localJointOffset=<0.5 0 0>
     }]
   }
   Anim Dynamics - Bone: dyn_shirtSleeve_top_r(Alpha=1, ExternalForce="0,0,0") {
    angularDampingOverride=0.9
    angularSpringConstant=5000
    bAngularSpring=true
    bGravityOverrideInSimSpace=true
    bLinearSpring=true
    bOverrideAngularDamping=true
    bOverrideLinearDamping=true
    bUseGravityOverride=true
    linearDampingOverride=1
    linearSpringConstant=40
    simulationSpace=RootRelative
    alphaBoolBlend {blendOption=Linear}
    boundBone {boneName=dyn_shirtSleeve_top_r}
    componentAppliedLinearAccClamp=<1000 1000 1000>
    componentLinearAccScale=<0.5 0.5 0.5>
    componentLinearVelScale=<0.5 0.5 0.5>
    gravityOverride {z=-500}
    physicsBodyDefinitions [{
      boundBone {boneName=dyn_shirtSleeve_top_r}
      boxExtents=<12 12 12>
      constraintSetup {angularLimitsMax=<0 2 10>, angularLimitsMin=<0 -30 -5>, angularTarget=<0 0 0>, linearAxesMax=<1 0.5 0>, linearAxesMin=<-2.5 -0.5 0>}
      localJointOffset=<0.5 0 0>
     }]
   }
   n4 = Anim Dynamics - Bone: dyn_shirtSleeve_bot_r(Alpha=1, ExternalForce="0,0,0") {
    angularDampingOverride=0.9
    angularSpringConstant=5000
    bAngularSpring=true
    bGravityOverrideInSimSpace=true
    bLinearSpring=true
    bOverrideAngularDamping=true
    bOverrideLinearDamping=true
    bUseGravityOverride=true
    linearDampingOverride=1
    linearSpringConstant=40
    simulationSpace=RootRelative
    alphaBoolBlend {blendOption=Linear}
    boundBone {boneName=dyn_shirtSleeve_bot_r}
    componentAppliedLinearAccClamp=<1000 1000 1000>
    componentLinearAccScale=<0.5 0.5 0.5>
    componentLinearVelScale=<0.5 0.5 0.5>
    gravityOverride {z=-500}
    physicsBodyDefinitions [{
      boundBone {boneName=dyn_shirtSleeve_bot_r}
      boxExtents=<12 12 12>
      constraintSetup {angularLimitsMax=<0 2 5>, angularLimitsMin=<0 -30 -5>, angularTarget=<0 0 0>, linearAxesMax=<1 0.5 0>, linearAxesMin=<-2.5 -0.5 0>}
      localJointOffset=<0.5 0 0>
     }]
   }
   Anim Dynamics - Bone: dyn_shirtSleeve_bck_r(ComponentPose=@n4, Alpha=1, ExternalForce="0,0,0") {
    angularDampingOverride=0.9
    angularSpringConstant=5000
    bAngularSpring=true
    bGravityOverrideInSimSpace=true
    bLinearSpring=true
    bOverrideAngularDamping=true
    bOverrideLinearDamping=true
    bUseGravityOverride=true
    linearDampingOverride=1
    linearSpringConstant=40
    simulationSpace=RootRelative
    alphaBoolBlend {blendOption=Linear}
    boundBone {boneName=dyn_shirtSleeve_bck_r}
    componentAppliedLinearAccClamp=<1000 1000 1000>
    componentLinearAccScale=<0.5 0.5 0.5>
    componentLinearVelScale=<0.5 0.5 0.5>
    gravityOverride {z=-500}
    physicsBodyDefinitions [{
      boundBone {boneName=dyn_shirtSleeve_bck_r}
      boxExtents=<12 12 12>
      constraintSetup {angularLimitsMax=<0 2 10>, angularLimitsMin=<0 -30 -10>, angularTarget=<0 0 0>, linearAxesMax=<1 0.5 0>, linearAxesMin=<-2.5 -0.5 0>}
      localJointOffset=<0.5 0 0>
     }]
   }
   "Constraint - dyn_compact_base (PPPPPPPP)"!(
    ComponentPose
    Alpha=1
    ConstraintWeights_0=0.2173
    ConstraintWeights_1:float=0.0151
    ConstraintWeights_2:float=0.0581
    ConstraintWeights_3:float=0.2639
    ConstraintWeights_4:float=0.2389
    ConstraintWeights_5:float=0.0001
    ConstraintWeights_6:float=0.1917
    ConstraintWeights_7:float=0.0148
   ) -> (Pose) {
    alphaBoolBlend {blendOption=Linear}
    boneToModify {boneName=dyn_compact_base}
    constraintSetup [
     {targetBone {boneName=spine_02}}
     {targetBone {boneName=spine_04}}
     {targetBone {boneName=spine_03}}
     {targetBone {boneName=spine_01}}
     {targetBone {boneName=pelvis}}
     {targetBone {boneName=thigh_twist_01_r}}
     {targetBone {boneName=thigh_r}}
     {targetBone {boneName=thigh_l}}
    ]
    constraintWeights [1,1,1,1,1,1,1,1]
   }
   n5 = Control Rig(Source=Copy Pose From Mesh() {bUseAttachedParent=true}) {
    alphaBoolBlend {blendOption=Linear}
    controlRigAssetReference {blueprintRigClass=&/F_MED_StoutWhiz_CtrlRig.F_MED_StoutWhiz_CtrlRig_C}
   }
   n6 = RigidBody!(
    ComponentPose=Local To Component(LocalPose=@n5)
    Alpha
    ExternalForce
    SimSpaceSettings:Sim Space Settings Structure=(WorldAlpha=0.2,VelocityScaleZ=1,DampingAlpha=1,MaxLinearVelocity=800,MaxAngularVelocity=8,MaxLinearAcceleration=10000,MaxAngularAcceleration=10000,ExternalLinearDragV=(X=0,Y=0,Z=0),ExternalLinearVelocity=(X=0,Y=0,Z=0),ExternalAngularVelocity=(X=0,Y=0,Z=0))
   ) -> (Pose) {
    bOverrideWorldGravity=true
    overridePhysicsAsset=&/F_MED_StoutWhiz_Physics
    simulationSpace=BaseBoneSpace
    alphaBoolBlend {blendOption=Linear}
    baseBoneRef {boneName=pelvis}
    overrideWorldGravity {z=-980}
    simSpaceSettings {maxAngularVelocity=8, maxLinearVelocity=800, worldAlpha=0.2}
   }
   n7 = "Constraint - dyn_overallStrap_c_l (PP)"!(ComponentPose=@n6, Alpha=1, ConstraintWeights_0=0.5, ConstraintWeights_1:float=0.5) -> (Pose) {
    alphaBoolBlend {blendOption=Linear}
    boneToModify {boneName=dyn_overallStrap_c_l}
    constraintSetup [{targetBone {boneName=dyn_overallStrap_c_l}}, {targetBone {boneName=dyn_overallsBack_l_05}}]
    constraintWeights [1,1]
   }
   n8 = "Constraint - dyn_overallStrap_c_r (PP)"!(ComponentPose=@n7, Alpha=1, ConstraintWeights_0=0.5, ConstraintWeights_1:float=0.5) -> (Pose) {
    alphaBoolBlend {blendOption=Linear}
    boneToModify {boneName=dyn_overallStrap_c_r}
    constraintSetup [{targetBone {boneName=dyn_overallsFront_r_04}}, {targetBone {boneName=dyn_overallsBack_r_05}}]
    constraintWeights [1,1]
   }
   n9 = "Constraint - dyn_banana_base (P)"(ComponentPose=@n8, Alpha=1, ConstraintWeights_0=1) {
    alphaBoolBlend {blendOption=Linear}
    boneToModify {boneName=dyn_banana_base}
    constraintSetup [{targetBone {boneName=dyn_overallsFront_r_03}}]
    constraintWeights [1]
   }
   n10 = Anim Dynamics - Bone: dyn_compact(ComponentPose=@n9, Alpha=1, ExternalForce="0,0,0") {
    angularSpringConstant=200
    bAngularSpring=true
    bUseGravityOverride=true
    simulationSpace=RootRelative
    alphaBoolBlend {blendOption=Linear}
    boundBone {boneName=dyn_compact}
    componentAppliedLinearAccClamp=<1000 1000 1000>
    componentLinearAccScale=<1 1 1>
    componentLinearVelScale=<1 1 1>
    gravityOverride {z=-980}
    physicsBodyDefinitions [{
      boundBone {boneName=dyn_compact}
      boxExtents=<5 5 5>
      constraintSetup {angularLimitsMax=<0 1 20>, angularLimitsMin=<0 -50 -1>, angularTarget=<0 0 0>}
      localJointOffset=<1 0 0>
     }]
    simSpaceSettings {simSpaceAngularAlpha=0.1}
   }
   n11 = Anim Dynamics - Bone: dyn_compactBot(ComponentPose=@n10, Alpha=1, ExternalForce="0,0,0") {
    angularSpringConstant=200
    bAngularSpring=true
    bUseGravityOverride=true
    simulationSpace=RootRelative
    alphaBoolBlend {blendOption=Linear}
    boundBone {boneName=dyn_compactBot}
    componentAppliedLinearAccClamp=<1000 1000 1000>
    componentLinearAccScale=<1 1 1>
    componentLinearVelScale=<1 1 1>
    gravityOverride {z=-980}
    physicsBodyDefinitions [{
      boundBone {boneName=dyn_compactBot}
      boxExtents=<5 5 5>
      constraintSetup {angularLimitsMax=<0 5 0>, angularLimitsMin=<0 -50 0>, angularTarget=<0 0 0>}
      localJointOffset=<1 0 0>
     }]
    simSpaceSettings {simSpaceAngularAlpha=1}
   }
   n12 = Anim Dynamics - Bone: dyn_banana_lowerBody(ComponentPose=@n11, Alpha=1, ExternalForce="0,0,0") {
    angularSpringConstant=200
    bAngularSpring=true
    simulationSpace=RootRelative
    alphaBoolBlend {blendOption=Linear}
    boundBone {boneName=dyn_banana_lowerBody}
    componentAppliedLinearAccClamp=<1000 1000 1000>
    componentLinearAccScale=<1 1 1>
    componentLinearVelScale=<1 1 1>
    physicsBodyDefinitions [{
      boundBone {boneName=dyn_banana_lowerBody}
      boxExtents=<10 10 10>
      constraintSetup {angularLimitsMax=<0 10 0>, angularLimitsMin=<0 -20 0>, angularTarget=<0 0 0>}
      localJointOffset=<0.1 0 0>
     }]
   }
   n13 = Anim Dynamics - Bone: dyn_banana_arm_r(ComponentPose=@n12, Alpha=1, ExternalForce="0,0,0") {
    angularSpringConstant=200
    bAngularSpring=true
    simulationSpace=RootRelative
    alphaBoolBlend {blendOption=Linear}
    boundBone {boneName=dyn_banana_arm_r}
    componentAppliedLinearAccClamp=<1000 1000 1000>
    componentLinearAccScale=<1 1 1>
    componentLinearVelScale=<1 1 1>
    physicsBodyDefinitions [{
      boundBone {boneName=dyn_banana_arm_r}
      boxExtents=<2 2 2>
      constraintSetup {angularLimitsMax=<0 10 5>, angularLimitsMin=<0 -20 -5>, angularTarget=<0 0 0>}
      localJointOffset=<0.1 0 0>
     }]
   }
   n14 = Anim Dynamics - Bone: dyn_banana_arm_l(ComponentPose=@n13, Alpha=1, ExternalForce="0,0,0") {
    angularSpringConstant=200
    bAngularSpring=true
    simulationSpace=RootRelative
    alphaBoolBlend {blendOption=Linear}
    boundBone {boneName=dyn_banana_arm_l}
    componentAppliedLinearAccClamp=<1000 1000 1000>
    componentLinearAccScale=<1 1 1>
    componentLinearVelScale=<1 1 1>
    physicsBodyDefinitions [{
      boundBone {boneName=dyn_banana_arm_l}
      boxExtents=<2 2 2>
      constraintSetup {angularLimitsMax=<0 10 5>, angularLimitsMin=<0 -20 -5>, angularTarget=<0 0 0>}
      localJointOffset=<0.1 0 0>
     }]
   }
   n15 = Anim Dynamics - Bone: dyn_banana_leg_r(ComponentPose=@n14, Alpha=1, ExternalForce="0,0,0") {
    angularSpringConstant=200
    bAngularSpring=true
    simulationSpace=RootRelative
    alphaBoolBlend {blendOption=Linear}
    boundBone {boneName=dyn_banana_leg_r}
    componentAppliedLinearAccClamp=<1000 1000 1000>
    componentLinearAccScale=<1 1 1>
    componentLinearVelScale=<1 1 1>
    physicsBodyDefinitions [{
      boundBone {boneName=dyn_banana_leg_r}
      boxExtents=<2 2 2>
      constraintSetup {angularLimitsMax=<0 10 10>, angularLimitsMin=<0 -20 -10>, angularTarget=<0 0 0>}
      localJointOffset=<0.1 0 0>
     }]
   }
   n16 = Anim Dynamics - Bone: dyn_banana_leg_l(ComponentPose=@n15, Alpha=1, ExternalForce="0,0,0") {
    angularSpringConstant=200
    bAngularSpring=true
    simulationSpace=RootRelative
    alphaBoolBlend {blendOption=Linear}
    boundBone {boneName=dyn_banana_leg_l}
    componentAppliedLinearAccClamp=<1000 1000 1000>
    componentLinearAccScale=<1 1 1>
    componentLinearVelScale=<1 1 1>
    physicsBodyDefinitions [{
      boundBone {boneName=dyn_banana_leg_l}
      boxExtents=<2 2 2>
      constraintSetup {angularLimitsMax=<0 10 10>, angularLimitsMin=<0 -20 -10>, angularTarget=<0 0 0>}
      localJointOffset=<0.1 0 0>
     }]
   }
   n17 = Anim Dynamics - Bone: dyn_banana_wire_01(ComponentPose=@n16, Alpha=1, ExternalForce="0,0,0") {
    angularSpringConstant=200
    bAngularSpring=true
    simulationSpace=RootRelative
    alphaBoolBlend {blendOption=Linear}
    boundBone {boneName=dyn_banana_wire_01}
    componentAppliedLinearAccClamp=<1000 1000 1000>
    componentLinearAccScale=<1 1 1>
    componentLinearVelScale=<1 1 1>
    physicsBodyDefinitions [{
      boundBone {boneName=dyn_banana_wire_01}
      boxExtents=<2 2 2>
      constraintSetup {angularLimitsMax=<0 5 10>, angularLimitsMin=<0 -5 -10>, angularTarget=<0 0 0>}
      localJointOffset=<0.1 0 0>
     }]
   }
   n18 = Anim Dynamics - Bone: dyn_banana_wire_02(ComponentPose=@n17, Alpha=1, ExternalForce="0,0,0") {
    angularSpringConstant=200
    bAngularSpring=true
    simulationSpace=RootRelative
    alphaBoolBlend {blendOption=Linear}
    boundBone {boneName=dyn_banana_wire_02}
    componentAppliedLinearAccClamp=<1000 1000 1000>
    componentLinearAccScale=<1 1 1>
    componentLinearVelScale=<1 1 1>
    physicsBodyDefinitions [{
      boundBone {boneName=dyn_banana_wire_02}
      boxExtents=<2 2 2>
      constraintSetup {angularLimitsMax=<0 5 10>, angularLimitsMin=<0 -5 -10>, angularTarget=<0 0 0>}
      localJointOffset=<0.1 0 0>
     }]
   }
   n19 = Anim Dynamics - Bone: dyn_shorts_fr_r!(ComponentPose=@n18, Alpha=1, LinearSpringConstant:float=50, ExternalForce="0,0,0") -> (Pose) {
    angularDampingOverride=0.9999
    angularSpringConstant=1000
    bAngularSpring=true
    bGravityOverrideInSimSpace=true
    bLinearSpring=true
    bOverrideAngularDamping=true
    bOverrideLinearDamping=true
    bUseGravityOverride=true
    linearDampingOverride=0.75
    linearSpringConstant=50
    simulationSpace=RootRelative
    alphaBoolBlend {blendOption=Linear}
    boundBone {boneName=dyn_shorts_fr_r}
    componentAppliedLinearAccClamp=<1000 1000 1000>
    componentLinearAccScale=<1 1 1>
    componentLinearVelScale=<1 1 1>
    gravityOverride {z=-980}
    physicsBodyDefinitions [{
      boundBone {boneName=dyn_shorts_fr_r}
      boxExtents=<14 14 14>
      constraintSetup {angularLimitsMax=<0 2 0>, angularLimitsMin=<0 -20 0>, angularTarget=<0 0 0>, linearAxesMax=<1 0.5 0>, linearAxesMin=<-2.5 -0.5 0>}
      localJointOffset=<1 0 0>
     }]
   }
   n20 = Anim Dynamics - Bone: dyn_shorts_in_r!(ComponentPose=@n19, Alpha=1, LinearSpringConstant:float=50, ExternalForce="0,0,0") -> (Pose) {
    angularDampingOverride=0.9999
    angularSpringConstant=1000
    bAngularSpring=true
    bGravityOverrideInSimSpace=true
    bLinearSpring=true
    bOverrideAngularDamping=true
    bOverrideLinearDamping=true
    bUseGravityOverride=true
    linearDampingOverride=0.75
    linearSpringConstant=50
    simulationSpace=RootRelative
    alphaBoolBlend {blendOption=Linear}
    boundBone {boneName=dyn_shorts_in_r}
    componentAppliedLinearAccClamp=<1000 1000 1000>
    componentLinearAccScale=<1 1 1>
    componentLinearVelScale=<1 1 1>
    gravityOverride {z=-980}
    physicsBodyDefinitions [{
      boundBone {boneName=dyn_shorts_in_r}
      boxExtents=<14 14 14>
      constraintSetup {angularLimitsMax=<0 2 5>, angularLimitsMin=<0 -20 -5>, angularTarget=<0 0 0>, linearAxesMax=<1 0.5 0>, linearAxesMin=<-2.5 -0.5 0>}
      localJointOffset=<1 0 0>
     }]
   }
   n21 = Anim Dynamics - Bone: dyn_shorts_out_r!(ComponentPose=@n20, Alpha=1, LinearSpringConstant:float=50, ExternalForce="0,0,0") -> (Pose) {
    angularDampingOverride=0.9999
    angularSpringConstant=1000
    bAngularSpring=true
    bGravityOverrideInSimSpace=true
    bLinearSpring=true
    bOverrideAngularDamping=true
    bOverrideLinearDamping=true
    bUseGravityOverride=true
    linearDampingOverride=0.75
    linearSpringConstant=50
    simulationSpace=RootRelative
    alphaBoolBlend {blendOption=Linear}
    boundBone {boneName=dyn_shorts_out_r}
    componentAppliedLinearAccClamp=<1000 1000 1000>
    componentLinearAccScale=<1 1 1>
    componentLinearVelScale=<1 1 1>
    gravityOverride {z=-980}
    physicsBodyDefinitions [{
      boundBone {boneName=dyn_shorts_out_r}
      boxExtents=<14 14 14>
      constraintSetup {angularLimitsMax=<0 2 5>, angularLimitsMin=<0 -20 -5>, angularTarget=<0 0 0>, linearAxesMax=<1 0.5 0>, linearAxesMin=<-2.5 -0.5 0>}
      localJointOffset=<1 0 0>
     }]
   }
   n22 = Anim Dynamics - Bone: dyn_shorts_bck_r!(ComponentPose=@n21, Alpha=1, LinearSpringConstant:float=50, ExternalForce="0,0,0") -> (Pose) {
    angularDampingOverride=0.9999
    angularSpringConstant=1000
    bAngularSpring=true
    bGravityOverrideInSimSpace=true
    bLinearSpring=true
    bOverrideAngularDamping=true
    bOverrideLinearDamping=true
    bUseGravityOverride=true
    linearDampingOverride=0.75
    linearSpringConstant=50
    simulationSpace=RootRelative
    alphaBoolBlend {blendOption=Linear}
    boundBone {boneName=dyn_shorts_bck_r}
    componentAppliedLinearAccClamp=<1000 1000 1000>
    componentLinearAccScale=<1 1 1>
    componentLinearVelScale=<1 1 1>
    gravityOverride {z=-980}
    physicsBodyDefinitions [{
      boundBone {boneName=dyn_shorts_bck_r}
      boxExtents=<14 14 14>
      constraintSetup {angularLimitsMax=<0 2 5>, angularLimitsMin=<0 -20 -5>, angularTarget=<0 0 0>}
      localJointOffset=<1 0 0>
     }]
   }
   n23 = Anim Dynamics - Bone: dyn_shorts_fr_l!(ComponentPose=@n22, Alpha=1, LinearSpringConstant:float=50, ExternalForce="0,0,0") -> (Pose) {
    angularDampingOverride=0.9999
    angularSpringConstant=1000
    bAngularSpring=true
    bGravityOverrideInSimSpace=true
    bLinearSpring=true
    bOverrideAngularDamping=true
    bOverrideLinearDamping=true
    bUseGravityOverride=true
    linearDampingOverride=0.75
    linearSpringConstant=50
    simulationSpace=RootRelative
    alphaBoolBlend {blendOption=Linear}
    boundBone {boneName=dyn_shorts_fr_l}
    componentAppliedLinearAccClamp=<1000 1000 1000>
    componentLinearAccScale=<1 1 1>
    componentLinearVelScale=<1 1 1>
    gravityOverride {z=-980}
    physicsBodyDefinitions [{
      boundBone {boneName=dyn_shorts_fr_l}
      boxExtents=<12 12 12>
      constraintSetup {angularLimitsMax=<0 10 0>, angularLimitsMin=<0 -20 0>, angularTarget=<0 0 0>, linearAxesMax=<1 0.5 0>, linearAxesMin=<-2.5 -0.5 0>}
      localJointOffset=<1 0 0>
     }]
   }
   n24 = Anim Dynamics - Bone: dyn_shorts_in_l!(ComponentPose=@n23, Alpha=1, LinearSpringConstant:float=50, ExternalForce="0,0,0") -> (Pose) {
    angularDampingOverride=0.9999
    angularSpringConstant=1000
    bAngularSpring=true
    bGravityOverrideInSimSpace=true
    bLinearSpring=true
    bOverrideAngularDamping=true
    bOverrideLinearDamping=true
    bUseGravityOverride=true
    linearDampingOverride=0.75
    linearSpringConstant=50
    simulationSpace=RootRelative
    alphaBoolBlend {blendOption=Linear}
    boundBone {boneName=dyn_shorts_in_l}
    componentAppliedLinearAccClamp=<1000 1000 1000>
    componentLinearAccScale=<1 1 1>
    componentLinearVelScale=<1 1 1>
    gravityOverride {z=-980}
    physicsBodyDefinitions [{
      boundBone {boneName=dyn_shorts_in_l}
      boxExtents=<12 12 12>
      constraintSetup {angularLimitsMax=<0 2 5>, angularLimitsMin=<0 -20 -5>, angularTarget=<0 0 0>, linearAxesMax=<1 0.5 0>, linearAxesMin=<-2.5 -0.5 0>}
      localJointOffset=<1 0 0>
     }]
   }
   n25 = Anim Dynamics - Bone: dyn_shorts_out_l!(ComponentPose=@n24, Alpha=1, LinearSpringConstant:float=50, ExternalForce="0,0,0") -> (Pose) {
    angularDampingOverride=0.9999
    angularSpringConstant=1000
    bAngularSpring=true
    bGravityOverrideInSimSpace=true
    bLinearSpring=true
    bOverrideAngularDamping=true
    bOverrideLinearDamping=true
    bUseGravityOverride=true
    linearDampingOverride=0.75
    linearSpringConstant=50
    simulationSpace=RootRelative
    alphaBoolBlend {blendOption=Linear}
    boundBone {boneName=dyn_shorts_out_l}
    componentAppliedLinearAccClamp=<1000 1000 1000>
    componentLinearAccScale=<1 1 1>
    componentLinearVelScale=<1 1 1>
    gravityOverride {z=-980}
    physicsBodyDefinitions [{
      boundBone {boneName=dyn_shorts_out_l}
      boxExtents=<12 12 12>
      constraintSetup {angularLimitsMax=<0 2 5>, angularLimitsMin=<0 -20 -5>, angularTarget=<0 0 0>, linearAxesMax=<1 0.5 0>, linearAxesMin=<-2.5 -0.5 0>}
      localJointOffset=<1 0 0>
     }]
   }
   n26 = Anim Dynamics - Bone: dyn_shorts_bck_l!(ComponentPose=@n25, Alpha=1, LinearSpringConstant:float=50, ExternalForce="0,0,0") -> (Pose) {
    angularDampingOverride=0.9999
    angularSpringConstant=1000
    bAngularSpring=true
    bGravityOverrideInSimSpace=true
    bLinearSpring=true
    bOverrideAngularDamping=true
    bOverrideLinearDamping=true
    bUseGravityOverride=true
    linearDampingOverride=0.75
    linearSpringConstant=50
    simulationSpace=RootRelative
    alphaBoolBlend {blendOption=Linear}
    boundBone {boneName=dyn_shorts_bck_l}
    componentAppliedLinearAccClamp=<1000 1000 1000>
    componentLinearAccScale=<1 1 1>
    componentLinearVelScale=<1 1 1>
    gravityOverride {z=-980}
    physicsBodyDefinitions [{
      boundBone {boneName=dyn_shorts_bck_l}
      boxExtents=<12 12 12>
      constraintSetup {angularLimitsMax=<0 2 5>, angularLimitsMin=<0 -20 -5>, angularTarget=<0 0 0>}
      localJointOffset=<1 0 0>
     }]
   }
   n27 = Anim Dynamics - Bone: dyn_shirtSleeve_top_r!(ComponentPose=@n26, Alpha=1, LinearSpringConstant:float=0, ExternalForce="0,0,0") -> (Pose) {
    angularDampingOverride=1
    angularSpringConstant=2000
    bAngularSpring=true
    bLinearSpring=true
    bOverrideAngularDamping=true
    bOverrideLinearDamping=true
    bUseGravityOverride=true
    gravityScale=0.35
    linearDampingOverride=1
    linearSpringConstant=50
    simulationSpace=RootRelative
    alphaBoolBlend {blendOption=Linear}
    boundBone {boneName=dyn_shirtSleeve_top_r}
    componentLinearAccScale=<1 1 1>
    componentLinearVelScale=<1 1 1>
    gravityOverride {z=-1500}
    physicsBodyDefinitions [{
      boundBone {boneName=dyn_shirtSleeve_top_r}
      boxExtents=<22 22 22>
      constraintSetup {angularLimitsMax=<10 10 10>, angularLimitsMin=<-10 -30 -10>, angularTarget=<1 0 0>, linearAxesMax=<1 0.5 0>, linearAxesMin=<-2.5 -0.5 0>}
      localJointOffset=<1 0 0>
     }]
   }
   n28 = Anim Dynamics - Bone: dyn_shirtSleeve_fr_r!(ComponentPose=@n27, Alpha=1, LinearSpringConstant:float=50, ExternalForce="0,0,0") -> (Pose) {
    angularDampingOverride=1
    angularSpringConstant=2000
    bAngularSpring=true
    bLinearSpring=true
    bOverrideAngularDamping=true
    bOverrideLinearDamping=true
    bUseGravityOverride=true
    gravityScale=0.35
    linearDampingOverride=1
    linearSpringConstant=50
    simulationSpace=RootRelative
    alphaBoolBlend {blendOption=Linear}
    boundBone {boneName=dyn_shirtSleeve_fr_r}
    componentLinearAccScale=<1 1 1>
    componentLinearVelScale=<1 1 1>
    gravityOverride {z=-1500}
    physicsBodyDefinitions [{
      boundBone {boneName=dyn_shirtSleeve_fr_r}
      boxExtents=<22 22 22>
      constraintSetup {angularLimitsMax=<10 3 10>, angularLimitsMin=<-10 -30 -10>, angularTarget=<1 0 0>, linearAxesMax=<1 0.5 0>, linearAxesMin=<-2.5 -0.5 0>}
      localJointOffset=<1 0 0>
     }]
   }
   n29 = Anim Dynamics - Bone: dyn_shirtSleeve_bot_r!(ComponentPose=@n28, Alpha=1, LinearSpringConstant:float=50, ExternalForce="0,0,0") -> (Pose) {
    angularDampingOverride=1
    angularSpringConstant=2000
    bAngularSpring=true
    bLinearSpring=true
    bOverrideAngularDamping=true
    bOverrideLinearDamping=true
    bUseGravityOverride=true
    gravityScale=0.35
    linearDampingOverride=1
    linearSpringConstant=50
    simulationSpace=RootRelative
    alphaBoolBlend {blendOption=Linear}
    boundBone {boneName=dyn_shirtSleeve_bot_r}
    componentLinearAccScale=<1 1 1>
    componentLinearVelScale=<1 1 1>
    gravityOverride {z=-1500}
    physicsBodyDefinitions [{
      boundBone {boneName=dyn_shirtSleeve_bot_r}
      boxExtents=<22 22 22>
      constraintSetup {angularLimitsMax=<10 30 10>, angularLimitsMin=<-10 -30 -10>, angularTarget=<1 0 0>, linearAxesMax=<1 0.5 0>, linearAxesMin=<-2.5 -0.5 0>}
      localJointOffset=<1 0 0>
     }]
   }
   n30 = Anim Dynamics - Bone: dyn_shirtSleeve_bck_r!(ComponentPose=@n29, Alpha=1, LinearSpringConstant:float=50, ExternalForce="0,0,0") -> (Pose) {
    angularDampingOverride=1
    angularSpringConstant=2000
    bAngularSpring=true
    bLinearSpring=true
    bOverrideAngularDamping=true
    bOverrideLinearDamping=true
    bUseGravityOverride=true
    gravityScale=0.35
    linearDampingOverride=1
    linearSpringConstant=50
    simulationSpace=RootRelative
    alphaBoolBlend {blendOption=Linear}
    boundBone {boneName=dyn_shirtSleeve_bck_r}
    componentLinearAccScale=<1 1 1>
    componentLinearVelScale=<1 1 1>
    gravityOverride {z=-1500}
    physicsBodyDefinitions [{
      boundBone {boneName=dyn_shirtSleeve_bck_r}
      boxExtents=<22 22 22>
      constraintSetup {angularLimitsMax=<10 30 10>, angularLimitsMin=<-10 -30 -10>, angularTarget=<1 0 0>, linearAxesMax=<1 0.5 0>, linearAxesMin=<-2.5 -0.5 0>}
      localJointOffset=<1 0 0>
     }]
   }
   n31 = Anim Dynamics - Bone: dyn_shirtSleeve_top_l!(ComponentPose=@n30, Alpha=1, LinearSpringConstant:float=50, ExternalForce="0,0,0") -> (Pose) {
    angularDampingOverride=1
    angularSpringConstant=2000
    bAngularSpring=true
    bLinearSpring=true
    bOverrideAngularDamping=true
    bOverrideLinearDamping=true
    bUseGravityOverride=true
    gravityScale=0.35
    linearDampingOverride=1
    linearSpringConstant=50
    simulationSpace=RootRelative
    alphaBoolBlend {blendOption=Linear}
    boundBone {boneName=dyn_shirtSleeve_top_l}
    componentLinearAccScale=<1 1 1>
    componentLinearVelScale=<1 1 1>
    gravityOverride {z=-1500}
    physicsBodyDefinitions [{
      boundBone {boneName=dyn_shirtSleeve_top_l}
      boxExtents=<23 23 23>
      constraintSetup {angularLimitsMax=<10 10 10>, angularLimitsMin=<-10 -30 -10>, angularTarget=<1 0 0>, linearAxesMax=<1 0.5 0>, linearAxesMin=<-2.5 -0.5 0>}
      localJointOffset=<1 0 0>
     }]
   }
   n32 = Anim Dynamics - Bone: dyn_shirtSleeve_fr_l!(ComponentPose=@n31, Alpha=1, LinearSpringConstant:float=50, ExternalForce="0,0,0") -> (Pose) {
    angularDampingOverride=1
    angularSpringConstant=2000
    bAngularSpring=true
    bLinearSpring=true
    bOverrideAngularDamping=true
    bOverrideLinearDamping=true
    bUseGravityOverride=true
    gravityScale=0.35
    linearDampingOverride=1
    linearSpringConstant=50
    simulationSpace=RootRelative
    alphaBoolBlend {blendOption=Linear}
    boundBone {boneName=dyn_shirtSleeve_fr_l}
    componentLinearAccScale=<1 1 1>
    componentLinearVelScale=<1 1 1>
    gravityOverride {z=-1500}
    physicsBodyDefinitions [{
      boundBone {boneName=dyn_shirtSleeve_fr_l}
      boxExtents=<23 23 23>
      constraintSetup {angularLimitsMax=<10 3 10>, angularLimitsMin=<-10 -30 -10>, angularTarget=<1 0 0>, linearAxesMax=<1 0.5 0>, linearAxesMin=<-2.5 -0.5 0>}
      localJointOffset=<1 0 0>
     }]
   }
   n33 = Anim Dynamics - Bone: dyn_shirtSleeve_bot_l!(ComponentPose=@n32, Alpha=1, LinearSpringConstant:float=50, ExternalForce="0,0,0") -> (Pose) {
    angularDampingOverride=1
    angularSpringConstant=2000
    bAngularSpring=true
    bLinearSpring=true
    bOverrideAngularDamping=true
    bOverrideLinearDamping=true
    bUseGravityOverride=true
    gravityScale=0.35
    linearDampingOverride=1
    linearSpringConstant=50
    simulationSpace=RootRelative
    alphaBoolBlend {blendOption=Linear}
    boundBone {boneName=dyn_shirtSleeve_bot_l}
    componentLinearAccScale=<1 1 1>
    componentLinearVelScale=<1 1 1>
    gravityOverride {z=-1500}
    physicsBodyDefinitions [{
      boundBone {boneName=dyn_shirtSleeve_bot_l}
      boxExtents=<23 23 23>
      constraintSetup {angularLimitsMax=<10 30 10>, angularLimitsMin=<-10 -30 -10>, angularTarget=<1 0 0>, linearAxesMax=<1 0.5 0>, linearAxesMin=<-2.5 -0.5 0>}
      localJointOffset=<1 0 0>
     }]
   }
   n34 = Anim Dynamics - Bone: dyn_shirtSleeve_bck_l!(ComponentPose=@n33, Alpha=1, LinearSpringConstant:float=50, ExternalForce="0,0,0") -> (Pose) {
    angularDampingOverride=1
    angularSpringConstant=2000
    bAngularSpring=true
    bLinearSpring=true
    bOverrideAngularDamping=true
    bOverrideLinearDamping=true
    bUseGravityOverride=true
    gravityScale=0.35
    linearDampingOverride=1
    linearSpringConstant=50
    simulationSpace=RootRelative
    alphaBoolBlend {blendOption=Linear}
    boundBone {boneName=dyn_shirtSleeve_bck_l}
    componentLinearAccScale=<1 1 1>
    componentLinearVelScale=<1 1 1>
    gravityOverride {z=-1500}
    physicsBodyDefinitions [{
      boundBone {boneName=dyn_shirtSleeve_bck_l}
      boxExtents=<23 23 23>
      constraintSetup {angularLimitsMax=<10 30 10>, angularLimitsMin=<-10 -30 -10>, angularTarget=<1 0 0>, linearAxesMax=<1 0.5 0>, linearAxesMin=<-2.5 -0.5 0>}
      localJointOffset=<1 0 0>
     }]
   }
   n35 = Anim Dynamics - Bone: dyn_hoodieString_r_01(ComponentPose=@n34, Alpha=1, ExternalForce="0,0,0") {
    angularDampingOverride=0.9999
    angularSpringConstant=200
    bAngularSpring=true
    bOverrideAngularDamping=true
    simulationSpace=RootRelative
    alphaBoolBlend {blendOption=Linear}
    boundBone {boneName=dyn_hoodieString_r_01}
    componentAppliedLinearAccClamp=<1000 1000 1000>
    componentLinearAccScale=<1 1 1>
    componentLinearVelScale=<1 1 1>
    physicsBodyDefinitions [{
      boundBone {boneName=dyn_hoodieString_r_01}
      boxExtents=<8 8 8>
      constraintSetup {angularLimitsMax=<0 0 10>, angularLimitsMin=<0 -20 -10>, angularTarget=<0 0 0>}
      localJointOffset=<0.5 0 0>
     }]
    simSpaceSettings {simSpaceAngularAlpha=0.2}
   }
   n36 = Anim Dynamics - Bone: dyn_hoodieString_r_02(ComponentPose=@n35, Alpha=1, ExternalForce="0,0,0") {
    angularDampingOverride=0.9999
    angularSpringConstant=200
    bAngularSpring=true
    bOverrideAngularDamping=true
    simulationSpace=RootRelative
    alphaBoolBlend {blendOption=Linear}
    boundBone {boneName=dyn_hoodieString_r_02}
    componentAppliedLinearAccClamp=<1000 1000 1000>
    componentLinearAccScale=<1 1 1>
    componentLinearVelScale=<1 1 1>
    physicsBodyDefinitions [{
      boundBone {boneName=dyn_hoodieString_r_02}
      boxExtents=<8 8 8>
      constraintSetup {angularLimitsMax=<0 0 10>, angularLimitsMin=<0 -20 -10>, angularTarget=<0 0 0>}
      localJointOffset=<0.5 0 0>
     }]
    simSpaceSettings {simSpaceAngularAlpha=0.2}
   }
   n37 = Anim Dynamics - Bone: dyn_hoodieString_r_03(ComponentPose=@n36, Alpha=1, ExternalForce="0,0,0") {
    angularDampingOverride=0.9999
    angularSpringConstant=200
    bAngularSpring=true
    bOverrideAngularDamping=true
    simulationSpace=RootRelative
    alphaBoolBlend {blendOption=Linear}
    boundBone {boneName=dyn_hoodieString_r_03}
    componentAppliedLinearAccClamp=<1000 1000 1000>
    componentLinearAccScale=<1 1 1>
    componentLinearVelScale=<1 1 1>
    physicsBodyDefinitions [{
      boundBone {boneName=dyn_hoodieString_r_03}
      boxExtents=<8 8 8>
      constraintSetup {angularLimitsMax=<0 0 10>, angularLimitsMin=<0 -20 -10>, angularTarget=<0 0 0>}
      localJointOffset=<0.5 0 0>
     }]
    simSpaceSettings {simSpaceAngularAlpha=0.2}
   }
   n38 = Anim Dynamics - Bone: dyn_hoodieString_l_01(ComponentPose=@n37, Alpha=1, ExternalForce="0,0,0") {
    angularDampingOverride=0.9999
    angularSpringConstant=200
    bAngularSpring=true
    bOverrideAngularDamping=true
    simulationSpace=RootRelative
    alphaBoolBlend {blendOption=Linear}
    boundBone {boneName=dyn_hoodieString_l_01}
    componentAppliedLinearAccClamp=<1000 1000 1000>
    componentLinearAccScale=<1 1 1>
    componentLinearVelScale=<1 1 1>
    physicsBodyDefinitions [{
      boundBone {boneName=dyn_hoodieString_l_01}
      boxExtents=<10 10 10>
      constraintSetup {angularLimitsMax=<0 0 10>, angularLimitsMin=<0 -20 -10>, angularTarget=<0 0 0>}
      localJointOffset=<0.5 0 0>
     }]
    simSpaceSettings {simSpaceAngularAlpha=0.2}
   }
   n39 = Anim Dynamics - Bone: dyn_hoodieString_l_02(ComponentPose=@n38, Alpha=1, ExternalForce="0,0,0") {
    angularDampingOverride=0.9999
    angularSpringConstant=200
    bAngularSpring=true
    bOverrideAngularDamping=true
    simulationSpace=RootRelative
    alphaBoolBlend {blendOption=Linear}
    boundBone {boneName=dyn_hoodieString_l_02}
    componentAppliedLinearAccClamp=<1000 1000 1000>
    componentLinearAccScale=<1 1 1>
    componentLinearVelScale=<1 1 1>
    physicsBodyDefinitions [{
      boundBone {boneName=dyn_hoodieString_l_02}
      boxExtents=<10 10 10>
      constraintSetup {angularLimitsMax=<0 0 10>, angularLimitsMin=<0 -20 -10>, angularTarget=<0 0 0>}
      localJointOffset=<0.5 0 0>
     }]
    simSpaceSettings {simSpaceAngularAlpha=0.2}
   }
   n40 = Anim Dynamics - Bone: dyn_hoodieString_l_03(ComponentPose=@n39, Alpha=1, ExternalForce="0,0,0") {
    angularDampingOverride=0.9999
    angularSpringConstant=200
    bAngularSpring=true
    bOverrideAngularDamping=true
    simulationSpace=RootRelative
    alphaBoolBlend {blendOption=Linear}
    boundBone {boneName=dyn_hoodieString_l_03}
    componentAppliedLinearAccClamp=<1000 1000 1000>
    componentLinearAccScale=<1 1 1>
    componentLinearVelScale=<1 1 1>
    physicsBodyDefinitions [{
      boundBone {boneName=dyn_hoodieString_l_03}
      boxExtents=<10 10 10>
      constraintSetup {angularLimitsMax=<0 0 10>, angularLimitsMin=<0 -20 -10>, angularTarget=<0 0 0>}
      localJointOffset=<0.5 0 0>
     }]
    simSpaceSettings {simSpaceAngularAlpha=0.2}
   }
   n41 = Anim Dynamics - Bone: dyn_hoodieString_l_04(ComponentPose=@n40, Alpha=1, ExternalForce="0,0,0") {
    angularDampingOverride=0.9999
    angularSpringConstant=200
    bAngularSpring=true
    bOverrideAngularDamping=true
    simulationSpace=RootRelative
    alphaBoolBlend {blendOption=Linear}
    boundBone {boneName=dyn_hoodieString_l_04}
    componentAppliedLinearAccClamp=<1000 1000 1000>
    componentLinearAccScale=<1 1 1>
    componentLinearVelScale=<1 1 1>
    physicsBodyDefinitions [{
      boundBone {boneName=dyn_hoodieString_l_04}
      boxExtents=<10 10 10>
      constraintSetup {angularLimitsMax=<0 0 10>, angularLimitsMin=<0 -20 -10>, angularTarget=<0 0 0>}
      localJointOffset=<0.5 0 0>
     }]
    simSpaceSettings {simSpaceAngularAlpha=0.2}
   }
   n42 = "Transform (Modify) Bone - Bone: dyn_banana_base"(ComponentPose=@n41, Alpha=1, Translation="0,0,0", Rotation="-0.005,-10.000,-0.001", Scale="1,1,1") {
    rotationMode=BMM_Additive
    rotationSpace=BCS_BoneSpace
    translationMode=BMM_Additive
    translationSpace=BCS_BoneSpace
    alphaBoolBlend {blendOption=Linear}
    alphaScaleBiasClamp {bInterpResult=true}
    boneToModify {boneName=dyn_banana_base}
    rotation=<rot -0.005 -10 -0.001>
   }
   Output Pose(Result=Component To Local(ComponentPose=@n42)) {layerGroup=None}
  ]}
 EventGraph {nodes [
   Event OnBeginSkydiving()
   > Set Clothing Spring Constant(Clothing Spring Constant:float=1000) -> (Output_Get:float)
   Event OnEndSkydiving()
   > Set Clothing Spring Constant(Clothing Spring Constant=50)
   Event OnCharacterCustomizationLoaded()
   > n1 = Branch(Condition=Get bIsInFrontEnd())
   > Set Sim Space Settings(Sim Space Settings:Sim Space Settings Structure=Make Sim Space Settings!(
     WorldAlpha:float=0.1
     VelocityScaleZ:float=1
     DampingAlpha:float=1
     MaxLinearVelocity:float=800
     MaxAngularVelocity:float=8
     MaxLinearAcceleration:float=10000
     MaxAngularAcceleration:float=10000
     ExternalLinearDragV:Vector="0,0,0"
     ExternalLinearVelocity:Vector="0,0,0"
     ExternalAngularVelocity:Vector="0,0,0"
    ) -> (SimSpaceSettings:Sim Space Settings Structure) {is=title K2Node_MakeStruct}) -> (Output_Get:Sim Space Settings Structure)
   > Set Frontend Alpha(Frontend Alpha:double=0.3) -> (Output_Get:double)
   > Set FrontEndBanana(FrontEndBanana:double=1) -> (Output_Get:double)
   Set Sim Space Settings(
    execute=@n1.else
    Sim Space Settings=Make Sim Space Settings(
     WorldAlpha=0.2
     VelocityScaleZ=1
     DampingAlpha=1
     MaxLinearVelocity=800
     MaxAngularVelocity=8
     MaxLinearAcceleration=1000
     MaxAngularAcceleration=1000
     ExternalLinearDragV="0,0,0"
     ExternalLinearVelocity="0,0,0"
     ExternalAngularVelocity="0,0,0"
    )
   )
   > Set Frontend Alpha(Frontend Alpha=1)
   > Set FrontEndBanana(FrontEndBanana=0)
   Event BlueprintUpdateAnimation()
   Set FrontEndBanana(FrontEndBanana=0)
  ]}
]
variables [
 Clothing Spring Constant {type=number, default=50}
 Frontend Alpha {type=number}
 FrontEndBanana {type=number}
 Sim Space Settings {
  type=SimSpaceSettings
  default {
   dampingAlpha=1
   maxAngularAcceleration=10000
   maxAngularVelocity=1000
   maxLinearAcceleration=10000
   maxLinearVelocity=1000
   velocityScaleZ=1
   worldAlpha=0.2
   externalLinearDragV=<0 0 0>
  }
 }
]
