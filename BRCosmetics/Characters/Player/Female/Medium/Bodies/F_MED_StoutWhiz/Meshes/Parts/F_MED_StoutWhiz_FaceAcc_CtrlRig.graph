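# Control Rig for the F_MED_StoutWhiz face accessory that applies corrective transforms to the hair braid bones, including both braid bases, as the head moves.
# A spherical pose reader on the head drives HeadUp.
# Crouch Move Head Up multiplies remapped values of the head's rotation and translation relative to spine_04.
# Each value passes through an AlphaInterp, with interpolation off, to weight its own Modify Transforms node, applied in sequence.
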ControlRigBlueprint /BRCosmetics/Characters/Player/Female/Medium/Bodies/F_MED_StoutWhiz/Meshes/Parts/F_MED_StoutWhiz_FaceAcc_CtrlRig
graphs [
 RigVMFunctionLibrary {links [], nodes []}
 RigVMFunction_Sequence_1_ContainedGraph {nodes [
   n1 = RigVMFunctionEntryNode()
   RigVMFunction_Sequence = &RigVMFunction_Sequence(ExecuteContext=@n1.ExecuteContext)
   n2 = &RigVMFunction_Sequence(ExecuteContext=@RigVMFunction_Sequence.B)
   RigVMFunctionReturnNode(A=@RigVMFunction_Sequence.A, B=@n2.A, C=@n2.B)
  ]}
 RigVMModel {nodes [
   RigUnit_BeginExecution = &RigUnit_BeginExecution()
   RigVMFunction_Sequence = RigVMAggregateNode(ExecuteContext=@RigUnit_BeginExecution.ExecutePin)
   n1 = &RigUnit_SphericalPoseReader(
    ActiveRegionScaleFactors=(PositiveWidth=1,NegativeWidth=1,PositiveHeight=1,NegativeHeight=1)
    ActiveRegionSize=0.2
    Debug=(bDrawDebug=false,bDraw2D=False,bDrawLocalAxes=False,DebugScale=25,DebugSegments=20,DebugThickness=0.25)
    DriverAxis=(X=1,Y=0,Z=0)
    DriverItem=(Type=Bone,Name="head")
    FalloffRegionScaleFactors=(PositiveWidth=1.5,NegativeWidth=1,PositiveHeight=1,NegativeHeight=1)
    FalloffSize=0.32
    FlipHeightScaling=False
    FlipWidthScaling=False
    OptionalParentItem=(Type=Bone,Name="None")
    RotationOffset=(X=-64.889,Y=15.1763,Z=36.1456)
    ExecutePin=@RigVMFunction_Sequence.A
   )
   RigVMVariableNode(ExecuteContext=@RigVMFunction_Sequence.B, Value=@n1.OutputParam) {variable=HeadUp}
   n2 = &RigUnit_GetRelativeTransformForItem(Child=(Type=Bone,Name="head"), Parent=(Type=Bone,Name="spine_04"), bChildInitial=False, bParentInitial=false)
   n3 = RigVMDispatchNode(Struct=@n2.RelativeTransform.Translation) {template="DISPATCH_RigVMDispatch_BreakStruct(in Struct,out Elements)"}
   Absolute = &RigVMFunction_MathDoubleAbs(Value=&RigVMFunction_MathQuaternionToEuler(RotationOrder=ZYX, Value=@n2.RelativeTransform.Rotation).Result.Z)
   > RigVMVariableNode(Value=&RigVMFunction_MathDoubleMul(
     B=&RigVMFunction_MathDoubleRemap(SourceMaximum=4, SourceMinimum=1, TargetMaximum=1, TargetMinimum=0, Value=@n3.Elements.Z).Result
     A=&RigVMFunction_MathDoubleRemap(SourceMaximum=80, SourceMinimum=60, TargetMaximum=1, TargetMinimum=0, Value=@Absolute.Result).Result
    ).Result) {variable=Crouch Move Head Up}
   AlphaInterp = &RigVMFunction_AlphaInterp(
    InterpSpeedDecreasing=20
    InterpSpeedIncreasing=20
    Scale=1
    bClampResult=False
    bInterpResult=False
    Value=RigVMVariableNode() {variable=HeadUp}.Value
   )
   &RigUnit_ModifyTransforms(
    ItemToModify="((Item=(Type=Bone,Name=\"dyn_braid_r_base\"),Transform=(Rotation=(X=0,Y=0.5,Z=0,W=-0.866),Translation=(X=-2.3201,Y=0.0001,Z=5.2365),Scale3D=(X=1,Y=1,Z=1))),(Item=(Type=Bone,Name=\"dyn_braid_l_base\"),Transform=(Rotation=(X=0,Y=0,Z=0,W=1),Translation=(X=0,Y=0,Z ...(truncated)"
    WeightMinimum=0
    Weight=@AlphaInterp.Result
    ExecutePin=@RigVMFunction_Sequence.C
   )
   n4 = &RigVMFunction_AlphaInterp(
    InterpSpeedDecreasing=20
    InterpSpeedIncreasing=20
    Scale=1
    bClampResult=False
    bInterpResult=False
    Value=RigVMVariableNode() {variable=Crouch Move Head Up}.Value
   )
   > &RigUnit_ModifyTransforms(
    ItemToModify="((Item=(Type=Bone,Name=\"dyn_braid_r_base\"),Transform=(Rotation=(X=0.056,Y=-0.173,Z=-0.0668,W=0.9811),Translation=(X=-2.6903,Y=-0.0694,Z=5.198),Scale3D=(X=1,Y=1,Z=1))),(Item=(Type=Bone,Name=\"dyn_braid_l_base\"),Transform=(Rotation=(X=0,Y=0,Z=0,W=1),Translati ...(truncated)"
    WeightMinimum=0
    Weight=@n4.Result
   )
  ]}
]
hierarchy [
 root {loc=<0 0 0>, rot=<0 0 0 1>}
 pelvis {parent=root, loc=<0 -0.5301 98.0571>, rot=<0.0127 0.707 -0.0127 -0.707>}
 spine_01 {parent=pelvis, loc=<6.1208 -0.9328 0>, rot=<0 0 0.034 -0.9994>}
 spine_02 {parent=spine_01, loc=<6.5997 -0.2095 0>, rot=<0 0 0.0238 -0.9997>}
 spine_03 {parent=spine_02, loc=<6.4914 -0.1237 0>, rot=<0 0 0.0149 -0.9999>}
 spine_04 {parent=spine_03, loc=<6.9598 0.0794 0>, rot=<0 0 -0.0132 -0.9999>}
 spine_05 {parent=spine_04, loc=<6.9987 0 0>, rot=<0 0 -0.0015 -1>}
 neck_01 {parent=spine_05, loc=<11.3138 0.1529 0>, rot=<0 0 -0.2003 -0.9797>}
 neck_02 {parent=neck_01, loc=<4.5627 0 0>, rot=<0 0 0.0504 -0.9987>}
 head {parent=neck_02, loc=<5.3914 0 0>, rot=<0 0 0.115 -0.9934>}
 dyn_braid_l_base {parent=head, loc=<-0.562 -6.1117 -5.042>, rot=<0.0835 -0.6801 -0.5734 -0.4492>}
 dyn_braid_l_01 {parent=dyn_braid_l_base, loc=<0 0 0>, rot=<-0.0005 0.0006 -0.0011 -1>}
 dyn_braid_l_02 {parent=dyn_braid_l_01, loc=<4.9462 1.0661 0.0747>, rot=<-0.1996 0.0198 -0.2654 -0.943>}
 dyn_braid_l_03 {parent=dyn_braid_l_02, loc=<8.6922 -0.1236 0.2425>, rot=<0.0327 -0.0469 0.0072 -0.9983>}
 dyn_braid_l_04 {parent=dyn_braid_l_03, loc=<9.307 0.0662 -0.0064>, rot=<0.0056 -0.1329 0.0046 -0.9911>}
 dyn_braid_l_05 {parent=dyn_braid_l_04, loc=<9.0388 -0.0841 -0.186>, rot=<-0.0193 -0.0677 0.0071 -0.9975>}
 dyn_braid_l_06 {parent=dyn_braid_l_05, loc=<8.5977 -0.2048 -0.2237>, rot=<0.0436 0 0 -0.999>}
 dyn_braidStrand_l {parent=dyn_braid_l_02, loc=<-1.2645 -0.7991 -1.1202>, rot=<-0.3947 0.5557 0.6362 -0.3614>}
 dyn_braid_r_base {parent=head, loc=<-0.5658 -5.6551 5.0419>, rot=<0.4187 -0.5361 0.7266 0.0969>}
 dyn_braid_r_01 {parent=dyn_braid_r_base, loc=<0 0 0>, rot=<-0.0008 -0.001 0 -1>}
 dyn_braid_r_02 {parent=dyn_braid_r_01, loc=<5.3017 1.8906 0.8922>, rot=<0.0312 0.0209 -0.286 -0.9575>}
 dyn_braid_r_03 {parent=dyn_braid_r_02, loc=<9.4449 -0.549 -1.2275>, rot=<0.0601 -0.1382 -0.0194 -0.9884>}
 dyn_braid_r_04 {parent=dyn_braid_r_03, loc=<9.1754 0.0059 -0.6205>, rot=<-0.0701 -0.0824 -0.0201 -0.9939>}
 dyn_braid_r_05 {parent=dyn_braid_r_04, loc=<8.5528 0.2645 -0.6037>, rot=<0.0374 -0.0761 -0.0325 -0.9959>}
 dyn_braid_r_06 {parent=dyn_braid_r_05, loc=<8.5748 -0.0468 0.2626>, rot=<0.0872 0 0 -0.9962>}
 dyn_braidStrand_r {parent=dyn_braid_r_02, loc=<0.1435 -2.4997 -0.2655>, rot=<-0.5078 0.386 0.3759 -0.6722>}
 dyn_hairStrand_A {parent=head, loc=<5.6326 -3.1003 -9.1256>, rot=<0.2004 -0.7767 0.3783 -0.4621>}
 dyn_hairStrand_B {parent=head, loc=<13.2631 -2.3252 5.9659>, rot=<0.5986 -0.1193 0.7822 -0.125>}
 dyn_hairStrand_C {parent=head, loc=<10.1596 4.2914 6.8418>, rot=<0.5478 -0.175 0.799 -0.1758>}
 dyn_bangs_A {parent=head, loc=<5.597 4.9381 -7.9545>, rot=<0.1309 -0.8101 0.3001 -0.4864>}
 dyn_bangs_B {parent=head, loc=<8.9217 9.5305 -3.4441>, rot=<0.0531 -0.81 -0.5693 -0.1304>}
 dyn_bangs_C {parent=head, loc=<10.449 10.174 -2.0072>, rot=<-0.18 -0.7464 -0.6381 -0.0586>}
 dyn_bangs_D {parent=head, loc=<8.8884 10.664 2.1865>, rot=<0.0829 0.6465 0.7515 0.1022>}
 dyn_bangs_E {parent=head, loc=<5.34 6.5325 6.9056>, rot=<0.5422 -0.194 0.7889 -0.2145>}
 clavicle_l {parent=spine_05, loc=<5.4831 -0.1053 -3.2946>, rot=<0.0465 -0.7289 -0.009 -0.6829>}
 upperarm_l {parent=clavicle_l, loc=<9.8658 -1.2136 -0.4105>, rot=<0.0164 -0.3245 -0.0269 -0.9454>}
 lowerarm_l {parent=upperarm_l, loc=<26.465 -0.8345 0.259>, rot=<0.0036 0.0161 -0.2298 -0.9731>}
 clavicle_r {parent=spine_05, loc=<5.4831 -0.1053 3.2946>, rot=<0.7289 0.0465 -0.6829 0.009>}
 upperarm_r {parent=clavicle_r, loc=<-9.8658 1.2136 0.4105>, rot=<0.0164 -0.3245 -0.0269 -0.9454>}
 lowerarm_r {parent=upperarm_r, loc=<-26.465 0.8345 -0.259>, rot=<0.0036 0.0161 -0.2298 -0.9731>}
]
variables [
 Crouch Move Head Up {type=double, default=0}
 HeadUp {type=double, default=0}
]
