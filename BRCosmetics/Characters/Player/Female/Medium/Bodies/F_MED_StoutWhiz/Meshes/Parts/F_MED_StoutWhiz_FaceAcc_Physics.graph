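# Physics asset for the F_MED_StoutWhiz face accessory mesh, set up for two dynamic hair braids.
# Each braid is a capsule chain from a small base on the head through six segments, with limited swing and near-zero twist.
# The left chain uses symmetric swing limits, while the right uses asymmetric ones.
# Capsules on head, neck, spine, pelvis and arms supply body shapes; braid-base, clavicle and upper-arm constraints leave rotation free.
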PhysicsAsset /BRCosmetics/Characters/Player/Female/Medium/Bodies/F_MED_StoutWhiz/Meshes/Parts/F_MED_StoutWhiz_FaceAcc_Physics
bodies {
 clavicle_l [{shapeType=Capsule, length=7, radius=6, shapeName=clavicle_l_capsule, center=<4.8448 -1.5796 -2.4643>, rotation=<rot -90 -0.0001 0>}]
 clavicle_r [{shapeType=Capsule, length=6.5, radius=6.5, shapeName=clavicle_r_capsule, center=<-4.9374 0 2.1754>, rotation=<rot 90 0.0002 0>}]
 dyn_braid_l_01 [{
   shapeType=Capsule
   length=1.5
   radius=2.5
   shapeName=dyn_braid_l_01_capsule
   center=<3.542 0.3815 0.8486>
   rotation=<rot -79.9999 30.0001 -10.0002>
  }]
 dyn_braid_l_02 [{shapeType=Capsule, length=3, radius=3, shapeName=dyn_braid_l_02_capsule, center=<4.2781 0 0.6099>, rotation=<rot -90 0 0>}]
 dyn_braid_l_03 [{shapeType=Capsule, length=4, radius=3, shapeName=dyn_braid_l_03_capsule, center=<3.4824 0 0.4752>, rotation=<rot -90 0 0>}]
 dyn_braid_l_04 [{shapeType=Capsule, length=3.5, radius=3, shapeName=dyn_braid_l_04_capsule, center=<3.4599 0.1883 0.7723>, rotation=<rot -90 0.0001 0>}]
 dyn_braid_l_05 [{shapeType=Capsule, length=3, radius=3.5, shapeName=dyn_braid_l_05_capsule, center=<3.1939 0 0.6737>, rotation=<rot -90 0 0>}]
 dyn_braid_l_06 [{shapeType=Capsule, length=5.7707, radius=5.102, shapeName=dyn_braid_l_06_capsule, center=<6.4869 0 -0.4727>, rotation=<rot -80 -180 -180>}]
 dyn_braid_l_base [{shapeType=Capsule, length=0.505, radius=0.505, shapeName=dyn_braid_l_base_capsule}]
 dyn_braid_r_01 [{shapeType=Capsule, length=2.5, radius=2.5, shapeName=dyn_braid_r_01_capsule, center=<3.6311 0.1118 0.4086>, rotation=<rot -90 30.0001 0>}]
 dyn_braid_r_02 [{shapeType=Capsule, length=3.5, radius=3, shapeName=dyn_braid_r_02_capsule, center=<4.9973 0 0.696>, rotation=<rot -80 -179.9999 180>}]
 dyn_braid_r_03 [{shapeType=Capsule, length=3.5, radius=3, shapeName=dyn_braid_r_03_capsule, center=<4.4739 0 1.146>, rotation=<rot -85 -179.9997 179.9999>}]
 dyn_braid_r_04 [{shapeType=Capsule, length=4, radius=3, shapeName=dyn_braid_r_04_capsule, center=<3.7457 0 0.9979>, rotation=<rot -90 0 0>}]
 dyn_braid_r_05 [{shapeType=Capsule, length=3, radius=3.5, shapeName=dyn_braid_r_05_capsule, center=<3.8212 0.2076 0.6875>, rotation=<rot -80 0.0001 0>}]
 dyn_braid_r_06 [{shapeType=Capsule, length=6.232, radius=4.9229, shapeName=dyn_braid_r_06_capsule, center=<6.2754 0 1.1065>, rotation=<rot -80 0.0001 0>}]
 dyn_braid_r_base [{shapeType=Capsule, length=0.505, radius=0.505, shapeName=dyn_braid_r_base_capsule}]
 head [{shapeType=Capsule, length=1.5, radius=5.5, shapeName=head_capsule, rotation=<rot -90 -0.0001 0>}]
 lowerarm_l [{shapeType=Capsule, length=15, radius=8, shapeName=lowerarm_l_capsule, center=<7.4652 0 0>, rotation=<rot 90 0.0003 0>}]
 lowerarm_r [{shapeType=Capsule, length=15, radius=7.5, shapeName=lowerarm_r_capsule, center=<-7.4652 0 0>, rotation=<rot -90 0 0>}]
 neck_01 [{shapeType=Capsule, length=50.505, radius=6.5, shapeName=neck_01_capsule, center=<-1.2686 -11.545 -6.9808>}]
 neck_02 [{shapeType=Capsule, length=51, radius=5, shapeName=neck_02_capsule, center=<-3.1648 0 8.6115>}]
 pelvis [{shapeType=Capsule, length=11, radius=11.5, shapeName=pelvis_capsule, center=<0 -1.9977 0>}]
 spine_01 [{shapeType=Capsule, length=13.5, radius=10.5, shapeName=spine_01_capsule, center=<0 0 -1.6796>}]
 spine_02 [{shapeType=Capsule, length=5, radius=8, shapeName=spine_02_capsule, center=<0 1.5555 0>}]
 spine_03 [{shapeType=Capsule, length=5, radius=10.5, shapeName=spine_03_capsule, center=<1.0717 2.1208 0>}]
 spine_04 [{shapeType=Capsule, length=9.5, radius=12, shapeName=spine_04_capsule, center=<1.0645 0 0>}]
 spine_05 [{shapeType=Capsule, length=13, radius=7, shapeName=spine_05_capsule, center=<6.6032 -2.7288 0>}]
 upperarm_l [{shapeType=Capsule, length=18, radius=6, shapeName=upperarm_l_capsule, center=<10.9288 -2.131 -1.2216>, rotation=<rot -80 -9.9989 9.999>}]
 upperarm_r [{
   shapeType=Capsule
   length=10.5
   radius=7.5
   shapeName=upperarm_r_capsule
   center=<-13.4823 -0.0001 0.4752>
   rotation=<rot 80 -179.9995 -179.9999>
  }]
}
constraints [
 {
  bone1Name=upperarm_l
  bone2Name=clavicle_l
  swing1LimitDegrees=45
  swing1Motion=Free
  swing2LimitDegrees=45
  swing2Motion=Free
  twistLimitDegrees=45
  twistMotion=Free
 }
 {
  bone1Name=upperarm_r
  bone2Name=clavicle_r
  swing1LimitDegrees=45
  swing1Motion=Free
  swing2LimitDegrees=45
  swing2Motion=Free
  twistLimitDegrees=45
  twistMotion=Free
 }
 {
  bone1Name=clavicle_l
  bone2Name=spine_05
  swing1LimitDegrees=45
  swing1Motion=Free
  swing2LimitDegrees=45
  swing2Motion=Free
  twistLimitDegrees=45
  twistMotion=Free
 }
 {
  bone1Name=clavicle_r
  bone2Name=spine_05
  swing1LimitDegrees=45
  swing1Motion=Free
  swing2LimitDegrees=45
  swing2Motion=Free
  twistLimitDegrees=45
  twistMotion=Free
 }
 {bone1Name=dyn_braid_l_02, bone2Name=dyn_braid_l_01, swing1LimitDegrees=40, swing2LimitDegrees=40, twistLimitDegrees=1}
 {bone1Name=dyn_braid_l_03, bone2Name=dyn_braid_l_02, swing1LimitDegrees=60, swing2LimitDegrees=60, twistLimitDegrees=1}
 {bone1Name=dyn_braid_l_04, bone2Name=dyn_braid_l_03, swing1LimitDegrees=60, swing2LimitDegrees=60, twistLimitDegrees=1}
 {bone1Name=dyn_braid_l_05, bone2Name=dyn_braid_l_04, swing1LimitDegrees=60, swing2LimitDegrees=60, twistLimitDegrees=1}
 {bone1Name=dyn_braid_l_01, bone2Name=dyn_braid_l_base, swing1LimitDegrees=25, swing2LimitDegrees=25, twistLimitDegrees=1}
 {bone1Name=dyn_braid_r_01, bone2Name=dyn_braid_r_base, swing1LimitDegrees=30, swing2LimitDegrees=90, twistLimitDegrees=1}
 {bone1Name=dyn_braid_r_02, bone2Name=dyn_braid_r_01, swing1LimitDegrees=30, swing2LimitDegrees=90, twistLimitDegrees=1}
 {bone1Name=dyn_braid_r_03, bone2Name=dyn_braid_r_02, swing1LimitDegrees=25, swing2LimitDegrees=90, twistLimitDegrees=1}
 {bone1Name=dyn_braid_r_04, bone2Name=dyn_braid_r_03, swing1LimitDegrees=25, swing2LimitDegrees=90, twistLimitDegrees=1}
 {bone1Name=dyn_braid_r_05, bone2Name=dyn_braid_r_04, swing1LimitDegrees=25, swing2LimitDegrees=90, twistLimitDegrees=1}
 {bone1Name=dyn_braid_l_06, bone2Name=dyn_braid_l_05, swing1LimitDegrees=60, swing2LimitDegrees=60, twistLimitDegrees=1}
 {bone1Name=dyn_braid_r_06, bone2Name=dyn_braid_r_05, swing1LimitDegrees=25, swing2LimitDegrees=45, twistLimitDegrees=1}
 {
  bone1Name=dyn_braid_l_base
  bone2Name=head
  swing1LimitDegrees=45
  swing1Motion=Free
  swing2LimitDegrees=45
  swing2Motion=Free
  twistLimitDegrees=45
  twistMotion=Free
 }
 {
  bone1Name=dyn_braid_r_base
  bone2Name=head
  swing1LimitDegrees=45
  swing1Motion=Free
  swing2LimitDegrees=45
  swing2Motion=Free
  twistLimitDegrees=45
  twistMotion=Free
 }
]
