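# Post-process style AnimBlueprint for the StoutWhiz face accessory part, animating the character's braids, bangs and hair strands.
# Copied parent and body-mesh poses blend per bone, then pass through Control Rig, RigidBody and per-bone Anim Dynamics into a cached pose.
# Weighted constraints to torso, neck and head bones plus additive offsets then shape the braids.
# Events set gravity, sim-space settings and alpha variables for front end, emotes, backbling and DBNO carry.
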AnimBlueprint /BRCosmetics/Characters/Player/Female/Medium/Bodies/F_MED_StoutWhiz/Meshes/Parts/F_MED_StoutWhiz_FaceAcc_AnimBP
target_skeleton=&/F_MED_StoutWhiz_FaceAcc_Skeleton
graphs [
 AnimGraph {nodes [
   n1 = "Constraint - dyn_braid_r_03 (PPPPP)"!(
    ComponentPose=Local To Component(LocalPose=Use cached pose 'DynCache'())
    Alpha=1
    ConstraintWeights_0=0.0003
    ConstraintWeights_1:float=0.9952
    ConstraintWeights_2:float=0.0026
    ConstraintWeights_3:float=0.0013
    ConstraintWeights_4:float=0.0006
   ) -> (Pose) {
    alphaBoolBlend {blendOption=Linear}
    boneToModify {boneName=dyn_braid_r_03}
    constraintSetup [
     {targetBone {boneName=spine_04}}
     {targetBone {boneName=spine_05}}
     {targetBone {boneName=clavicle_r}}
     {targetBone {boneName=neck_01}}
     {targetBone {boneName=head}}
    ]
    constraintWeights [1,1,1,1,1]
   }
   n2 = "Constraint - dyn_braid_r_04 (PPPPPPPP)"!(
    ComponentPose=@n1
    Alpha=1
    ConstraintWeights_0=0.0331
    ConstraintWeights_1:float=0.1313
    ConstraintWeights_2:float=0.1101
    ConstraintWeights_3:float=0.3984
    ConstraintWeights_4:float=0.0075
    ConstraintWeights_5:float=0.3046
    ConstraintWeights_6:float=0.0022
    ConstraintWeights_7:float=0.0128
   ) -> (Pose) {
    alphaBoolBlend {blendOption=Linear}
    boneToModify {boneName=dyn_braid_r_04}
    constraintSetup [
     {targetBone {boneName=spine_02}}
     {targetBone {boneName=spine_04}}
     {targetBone {boneName=spine_03}}
     {targetBone {boneName=spine_05}}
     {targetBone {boneName=clavicle_l}}
     {targetBone {boneName=clavicle_r}}
     {targetBone {boneName=spine_01}}
     {targetBone {boneName=neck_01}}
    ]
    constraintWeights [1,1,1,1,1,1,1,1]
   }
   n3 = "Constraint - dyn_braid_r_05 (PPPPPPPP)"!(
    ComponentPose=@n2
    Alpha=1
    ConstraintWeights_0=0.1427
    ConstraintWeights_1:float=0.1771
    ConstraintWeights_2:float=0.3525
    ConstraintWeights_3:float=0.1957
    ConstraintWeights_4:float=0.0012
    ConstraintWeights_5:float=0.1196
    ConstraintWeights_6:float=0.0101
    ConstraintWeights_7:float=0.0011
   ) -> (Pose) {
    alphaBoolBlend {blendOption=Linear}
    boneToModify {boneName=dyn_braid_r_05}
    constraintSetup [
     {targetBone {boneName=spine_02}}
     {targetBone {boneName=spine_04}}
     {targetBone {boneName=spine_03}}
     {targetBone {boneName=spine_05}}
     {targetBone {boneName=clavicle_l}}
     {targetBone {boneName=clavicle_r}}
     {targetBone {boneName=spine_01}}
     {targetBone {boneName=neck_01}}
    ]
    constraintWeights [1,1,1,1,1,1,1,1]
   }
   n4 = "Constraint - dyn_braid_r_06 (PPPPPP)"!(
    ComponentPose=@n3
    Alpha=1
    ConstraintWeights_0=0.3245
    ConstraintWeights_1:float=0.102
    ConstraintWeights_2:float=0.5301
    ConstraintWeights_3:float=0.0187
    ConstraintWeights_4:float=0.0042
    ConstraintWeights_5:float=0.0205
   ) -> (Pose) {
    alphaBoolBlend {blendOption=Linear}
    boneToModify {boneName=dyn_braid_r_06}
    constraintSetup [
     {targetBone {boneName=spine_02}}
     {targetBone {boneName=spine_04}}
     {targetBone {boneName=spine_03}}
     {targetBone {boneName=spine_05}}
     {targetBone {boneName=clavicle_r}}
     {targetBone {boneName=spine_01}}
    ]
    constraintWeights [1,1,1,1,1,1]
   }
   n5 = "Transform (Modify) Bone - Bone: dyn_braid_r_03"(ComponentPose=@n4, Alpha=1, Translation="0,0,0", Rotation="-10.023,10.003,-0.019", Scale="1,1,1") {
    rotationMode=BMM_Additive
    rotationSpace=BCS_BoneSpace
    translationMode=BMM_Additive
    alphaBoolBlend {blendOption=Linear}
    alphaScaleBiasClamp {bInterpResult=true}
    boneToModify {boneName=dyn_braid_r_03}
    rotation=<rot -10.023 10.003 -0.019>
   }
   n6 = "Transform (Modify) Bone - Bone: dyn_braid_r_06"(ComponentPose=@n5, Alpha=1, Translation="0,0,0", Rotation="19.704,2.761,0.805", Scale="1,1,1") {
    rotationMode=BMM_Additive
    rotationSpace=BCS_BoneSpace
    translationMode=BMM_Additive
    alphaBoolBlend {blendOption=Linear}
    alphaScaleBiasClamp {bInterpResult=true}
    boneToModify {boneName=dyn_braid_r_06}
    rotation=<rot 19.704 2.761 0.805>
   }
   n7 = "Transform (Modify) Bone - Bone: dyn_braid_r_base"(ComponentPose=@n6, Alpha=1, Translation="0,0,0", Rotation="34.773,52.476,-28.497", Scale="1,1,1") {
    rotationMode=BMM_Additive
    rotationSpace=BCS_BoneSpace
    translationMode=BMM_Additive
    alphaBoolBlend {blendOption=Linear}
    alphaScaleBiasClamp {bInterpResult=true}
    boneToModify {boneName=dyn_braid_r_base}
    rotation=<rot 34.773 52.476 -28.497>
   }
   n8 = "Constraint - dyn_braid_l_03 (PPPPPPPP)"!(
    ComponentPose=@n7
    Alpha=1
    ConstraintWeights_0=0.047
    ConstraintWeights_1:float=0.0035
    ConstraintWeights_2:float=0.2771
    ConstraintWeights_3:float=0.4956
    ConstraintWeights_4:float=0.0055
    ConstraintWeights_5:float=0.0788
    ConstraintWeights_6:float=0.0911
    ConstraintWeights_7:float=0.0013
   ) -> (Pose) {
    alphaBoolBlend {blendOption=Linear}
    boneToModify {boneName=dyn_braid_l_03}
    constraintSetup [
     {targetBone {boneName=spine_04}}
     {targetBone {boneName=spine_03}}
     {targetBone {boneName=spine_05}}
     {targetBone {boneName=clavicle_l}}
     {targetBone {boneName=clavicle_r}}
     {targetBone {boneName=neck_01}}
     {targetBone {boneName=head}}
     {targetBone {boneName=neck_02}}
    ]
    constraintWeights [1,1,1,1,1,1,1,1]
   }
   n9 = "Constraint - dyn_braid_r_01 (P)"(ComponentPose=@n8, Alpha=1, ConstraintWeights_0=1) {
    alphaBoolBlend {blendOption=Linear}
    alphaScaleBiasClamp {bInterpResult=true, interpSpeedDecreasing=5, interpSpeedIncreasing=5}
    boneToModify {boneName=dyn_braid_r_01}
    constraintSetup [{targetBone {boneName=spine_05}}]
    constraintWeights [1]
   }
   n10 = "Transform (Modify) Bone - Bone: dyn_braid_r_01"(ComponentPose=@n9, Alpha=1, Translation="0,0,-0.75", Rotation="-3,-0.001,0", Scale="1,1,1") {
    rotationMode=BMM_Additive
    rotationSpace=BCS_BoneSpace
    translationMode=BMM_Additive
    translationSpace=BCS_BoneSpace
    alphaBoolBlend {blendOption=Linear}
    alphaScaleBiasClamp {bInterpResult=true}
    boneToModify {boneName=dyn_braid_r_01}
    rotation {pitch=-3, yaw=-0.001}
    translation {z=-0.75}
   }
   n11 = "Transform (Modify) Bone - Bone: dyn_braid_r_03"(ComponentPose=@n10, Alpha=1, Translation="0,0,0", Rotation="13,0,8", Scale="1,1,1") {
    rotationMode=BMM_Additive
    rotationSpace=BCS_BoneSpace
    alphaBoolBlend {blendOption=Linear}
    alphaScaleBiasClamp {bInterpResult=true}
    boneToModify {boneName=dyn_braid_r_03}
    rotation {pitch=13, roll=8}
   }
   n12 = "Transform (Modify) Bone - Bone: dyn_braid_r_03"(ComponentPose=@n11, Alpha=1, Translation="0,0,0", Rotation="5,2.562,0", Scale="1,1,1") {
    rotationMode=BMM_Additive
    rotationSpace=BCS_BoneSpace
    translationMode=BMM_Additive
    alphaBoolBlend {blendOption=Linear}
    alphaScaleBiasClamp {bInterpResult=true}
    boneToModify {boneName=dyn_braid_r_03}
    rotation {pitch=5, yaw=2.562}
   }
   n13 = "Transform (Modify) Bone - Bone: dyn_braid_r_06"(ComponentPose=@n12, Alpha=1, Translation="0,0,2", Rotation="0,-7.397,-0.918", Scale="1,1,0.8") {
    rotationMode=BMM_Additive
    rotationSpace=BCS_BoneSpace
    scaleMode=BMM_Replace
    scaleSpace=BCS_BoneSpace
    translationMode=BMM_Additive
    alphaBoolBlend {blendOption=Linear}
    alphaScaleBiasClamp {bInterpResult=true}
    boneToModify {boneName=dyn_braid_r_06}
    rotation {roll=-0.918, yaw=-7.397}
    scale {z=0.8}
    translation {z=2}
   }
   Output Pose(Result=Component To Local(ComponentPose=@n13)) {layerGroup=None}
   n14 = Get Body Mesh() -> (Body Mesh:Skeletal Mesh Component Object Reference)
   n15 = Control Rig(Source=Layered blend per bone(
     BasePose=Copy Pose From Mesh() {bCopyCurves=true, bUseAttachedParent=true}
     BlendPoses_0=Copy Pose From Mesh(SourceMeshComponent=@n14) {bCopyCurves=true, bUseAttachedParent=true}
    ) {layerSetup [{branchFilters [{boneName=upperarm_l}, {boneName=upperarm_r}]}]}) {
    alphaBoolBlend {blendOption=Linear}
    controlRigAssetReference {blueprintRigClass=&/F_MED_StoutWhiz_FaceAcc_CtrlRig.F_MED_StoutWhiz_FaceAcc_CtrlRig_C}
   }
   n16 = RigidBody!(
    ComponentPose=Local To Component(LocalPose=@n15)
    Alpha
    OverrideWorldGravity:Vector="0,0,-800"
    ExternalForce
    SimSpaceSettings:Sim Space Settings Structure=(WorldAlpha=0,VelocityScaleZ=1,DampingAlpha=1,MaxLinearVelocity=10000,MaxAngularVelocity=10000,MaxLinearAcceleration=10000,MaxAngularAcceleration=10000,ExternalLinearDragV=(X=0,Y=0,Z=0),ExternalLinearVelocity=(X=0,Y=0,Z=0),ExternalAngularVelocity=(X=0,Y=0,Z=0))
   ) -> (Pose) {
    bClampLinearTranslationLimitToRefPose=true
    bOverrideWorldGravity=true
    overridePhysicsAsset=&/F_MED_StoutWhiz_FaceAcc_Physics
    simulationSpace=BaseBoneSpace
    alphaBoolBlend {blendOption=Linear}
    baseBoneRef {boneName=neck_02}
    componentAppliedLinearAccClamp=<3000 3000 3000>
    overrideWorldGravity {z=-800}
    simSpaceSettings {maxAngularVelocity=1000, maxLinearVelocity=1000, worldAlpha=0.3}
   }
   n17 = Anim Dynamics - Bone: dyn_bangs_A(ComponentPose=@n16, Alpha=1, ExternalForce="0,0,0") {
    angularDampingOverride=0.9999
    angularSpringConstant=4000
    bAngularSpring=true
    bOverrideAngularDamping=true
    bUseGravityOverride=true
    gravityScale=0.35
    simulationSpace=RootRelative
    alphaBoolBlend {blendOption=Linear}
    boundBone {boneName=dyn_bangs_A}
    componentAppliedLinearAccClamp=<1000 1000 1000>
    componentLinearAccScale=<1 1 1>
    componentLinearVelScale=<1 1 1>
    gravityOverride {z=-980}
    physicsBodyDefinitions [{
      boundBone {boneName=dyn_bangs_A}
      boxExtents=<13 13 13>
      constraintSetup {angularLimitsMax=<0 20 2>, angularLimitsMin=<0 -30 -2>, angularTarget=<1 0 0>}
      localJointOffset=<1 0 0>
     }]
    relativeSpaceBone {boneName=root}
    simSpaceSettings {simSpaceAngularAlpha=0.2}
   }
   n18 = Anim Dynamics - Bone: dyn_bangs_B(ComponentPose=@n17, Alpha=1, ExternalForce="0,0,0") {
    angularDampingOverride=0.9999
    angularSpringConstant=4000
    bAngularSpring=true
    bOverrideAngularDamping=true
    bUseGravityOverride=true
    gravityScale=0.35
    simulationSpace=RootRelative
    alphaBoolBlend {blendOption=Linear}
    boundBone {boneName=dyn_bangs_B}
    componentAppliedLinearAccClamp=<1000 1000 1000>
    componentLinearAccScale=<1 1 1>
    componentLinearVelScale=<1 1 1>
    gravityOverride {z=-980}
    physicsBodyDefinitions [{
      boundBone {boneName=dyn_bangs_B}
      boxExtents=<15 15 15>
      constraintSetup {angularLimitsMax=<0 1 10>, angularLimitsMin=<0 -30 -10>, angularTarget=<1 0 0>}
      localJointOffset=<1 0 0>
     }]
    relativeSpaceBone {boneName=root}
    simSpaceSettings {simSpaceAngularAlpha=0.2}
   }
   n19 = Anim Dynamics - Bone: dyn_bangs_C(ComponentPose=@n18, Alpha=1, ExternalForce="0,0,0") {
    angularDampingOverride=0.9999
    angularSpringConstant=4000
    bAngularSpring=true
    bOverrideAngularDamping=true
    bUseGravityOverride=true
    gravityScale=0.35
    simulationSpace=RootRelative
    alphaBoolBlend {blendOption=Linear}
    boundBone {boneName=dyn_bangs_C}
    componentAppliedLinearAccClamp=<1000 1000 1000>
    componentLinearAccScale=<1 1 1>
    componentLinearVelScale=<1 1 1>
    gravityOverride {z=-980}
    physicsBodyDefinitions [{
      boundBone {boneName=dyn_bangs_C}
      boxExtents=<14 14 14>
      constraintSetup {angularLimitsMax=<0 1 10>, angularLimitsMin=<0 -30 -10>, angularTarget=<1 0 0>}
      localJointOffset=<1 0 0>
     }]
    relativeSpaceBone {boneName=root}
    simSpaceSettings {simSpaceAngularAlpha=0.2}
   }
   n20 = Anim Dynamics - Bone: dyn_bangs_D(ComponentPose=@n19, Alpha=1, ExternalForce="0,0,0") {
    angularDampingOverride=0.9999
    angularSpringConstant=4000
    bAngularSpring=true
    bOverrideAngularDamping=true
    bUseGravityOverride=true
    gravityScale=0.35
    simulationSpace=RootRelative
    alphaBoolBlend {blendOption=Linear}
    boundBone {boneName=dyn_bangs_D}
    componentAppliedLinearAccClamp=<1000 1000 1000>
    componentLinearAccScale=<1 1 1>
    componentLinearVelScale=<1 1 1>
    gravityOverride {z=-980}
    physicsBodyDefinitions [{
      boundBone {boneName=dyn_bangs_D}
      boxExtents=<13 13 13>
      constraintSetup {angularLimitsMax=<0 1 10>, angularLimitsMin=<0 -30 -10>, angularTarget=<1 0 0>}
      localJointOffset=<1 0 0>
     }]
    relativeSpaceBone {boneName=root}
    simSpaceSettings {simSpaceAngularAlpha=0.2}
   }
   n21 = Anim Dynamics - Bone: dyn_bangs_E(ComponentPose=@n20, Alpha=1, ExternalForce="0,0,0") {
    angularDampingOverride=0.9999
    angularSpringConstant=4000
    bAngularSpring=true
    bOverrideAngularDamping=true
    bUseGravityOverride=true
    gravityScale=0.35
    simulationSpace=RootRelative
    alphaBoolBlend {blendOption=Linear}
    boundBone {boneName=dyn_bangs_E}
    componentAppliedLinearAccClamp=<1000 1000 1000>
    componentLinearAccScale=<1 1 1>
    componentLinearVelScale=<1 1 1>
    gravityOverride {z=-980}
    physicsBodyDefinitions [{
      boundBone {boneName=dyn_bangs_E}
      boxExtents=<17 17 17>
      constraintSetup {angularLimitsMax=<0 20 5>, angularLimitsMin=<0 -20 -20>, angularTarget=<1 0 0>}
      localJointOffset=<1 0 0>
     }]
    relativeSpaceBone {boneName=root}
    simSpaceSettings {simSpaceAngularAlpha=0.2}
   }
   n22 = Anim Dynamics - Bone: dyn_braidStrand_l(ComponentPose=@n21, Alpha=1, ExternalForce="0,0,0") {
    angularDampingOverride=0.9999
    angularSpringConstant=4000
    bAngularSpring=true
    bOverrideAngularDamping=true
    bUseGravityOverride=true
    gravityScale=0.35
    simulationSpace=RootRelative
    alphaBoolBlend {blendOption=Linear}
    boundBone {boneName=dyn_braidStrand_l}
    componentAppliedLinearAccClamp=<1000 1000 1000>
    componentLinearAccScale=<1 1 1>
    componentLinearVelScale=<1 1 1>
    gravityOverride {z=-980}
    physicsBodyDefinitions [{
      boundBone {boneName=dyn_braidStrand_l}
      boxExtents=<13 13 13>
      constraintSetup {angularLimitsMax=<0 20 2>, angularLimitsMin=<0 -30 -2>, angularTarget=<1 0 0>}
      localJointOffset=<1 0 0>
     }]
    relativeSpaceBone {boneName=root}
    simSpaceSettings {simSpaceAngularAlpha=0.2}
   }
   n23 = Anim Dynamics - Bone: dyn_braidStrand_r(ComponentPose=@n22, Alpha=1, ExternalForce="0,0,0") {
    angularDampingOverride=0.9999
    angularSpringConstant=4000
    bAngularSpring=true
    bOverrideAngularDamping=true
    bUseGravityOverride=true
    gravityScale=0.35
    simulationSpace=RootRelative
    alphaBoolBlend {blendOption=Linear}
    boundBone {boneName=dyn_braidStrand_r}
    componentAppliedLinearAccClamp=<1000 1000 1000>
    componentLinearAccScale=<1 1 1>
    componentLinearVelScale=<1 1 1>
    gravityOverride {z=-980}
    physicsBodyDefinitions [{
      boundBone {boneName=dyn_braidStrand_r}
      boxExtents=<15 15 15>
      constraintSetup {angularLimitsMax=<0 1 10>, angularLimitsMin=<0 -30 -10>, angularTarget=<1 0 0>}
      localJointOffset=<1 0 0>
     }]
    relativeSpaceBone {boneName=root}
    simSpaceSettings {simSpaceAngularAlpha=0.2}
   }
   n24 = Anim Dynamics - Bone: dyn_hairStrand_A(ComponentPose=@n23, Alpha=1, ExternalForce="0,0,0") {
    angularDampingOverride=0.9999
    angularSpringConstant=4000
    bAngularSpring=true
    bOverrideAngularDamping=true
    bUseGravityOverride=true
    gravityScale=0.35
    simulationSpace=RootRelative
    alphaBoolBlend {blendOption=Linear}
    boundBone {boneName=dyn_hairStrand_A}
    componentAppliedLinearAccClamp=<1000 1000 1000>
    componentLinearAccScale=<1 1 1>
    componentLinearVelScale=<1 1 1>
    gravityOverride {z=-980}
    physicsBodyDefinitions [{
      boundBone {boneName=dyn_hairStrand_A}
      boxExtents=<14 14 14>
      constraintSetup {angularLimitsMax=<0 1 10>, angularLimitsMin=<0 -30 -10>, angularTarget=<1 0 0>}
      localJointOffset=<1 0 0>
     }]
    relativeSpaceBone {boneName=root}
    simSpaceSettings {simSpaceAngularAlpha=0.2}
   }
   n25 = Anim Dynamics - Bone: dyn_hairStrand_B(ComponentPose=@n24, Alpha=1, ExternalForce="0,0,0") {
    angularDampingOverride=0.9999
    angularSpringConstant=3000
    bAngularSpring=true
    bOverrideAngularDamping=true
    bUseGravityOverride=true
    gravityScale=0.35
    simulationSpace=RootRelative
    alphaBoolBlend {blendOption=Linear}
    boundBone {boneName=dyn_hairStrand_B}
    componentAppliedLinearAccClamp=<1000 1000 1000>
    componentLinearAccScale=<1 1 1>
    componentLinearVelScale=<1 1 1>
    gravityOverride {z=-980}
    physicsBodyDefinitions [{
      boundBone {boneName=dyn_hairStrand_B}
      boxExtents=<15 15 15>
      constraintSetup {angularLimitsMax=<0 1 10>, angularLimitsMin=<0 -30 -10>, angularTarget=<1 0 0>}
      localJointOffset=<1 0 0>
     }]
    relativeSpaceBone {boneName=root}
    simSpaceSettings {simSpaceAngularAlpha=0.2}
   }
   n26 = Anim Dynamics - Bone: dyn_hairStrand_C(ComponentPose=@n25, Alpha=1, ExternalForce="0,0,0") {
    angularDampingOverride=0.9999
    angularSpringConstant=3000
    bAngularSpring=true
    bOverrideAngularDamping=true
    bUseGravityOverride=true
    gravityScale=0.35
    simulationSpace=RootRelative
    alphaBoolBlend {blendOption=Linear}
    boundBone {boneName=dyn_hairStrand_C}
    componentAppliedLinearAccClamp=<1000 1000 1000>
    componentLinearAccScale=<1 1 1>
    componentLinearVelScale=<1 1 1>
    gravityOverride {z=-980}
    physicsBodyDefinitions [{
      boundBone {boneName=dyn_hairStrand_C}
      boxExtents=<25 25 25>
      constraintSetup {angularLimitsMax=<0 20 5>, angularLimitsMin=<0 -20 -20>, angularTarget=<1 0 0>}
      localJointOffset=<1 0 0>
     }]
    relativeSpaceBone {boneName=root}
    simSpaceSettings {simSpaceAngularAlpha=0.2}
   }
   Save cached pose 'DynCache'(Pose=Component To Local(ComponentPose=@n26)) {cacheName=DynCache}
  ]}
 EventGraph {nodes [
   n1 = Event BlueprintUpdateAnimation()
   > n2 = "Gravity Override State Function (Parameter Asset) (Nativized)"(
    CharacterPartAnimInstance=Self-Reference()
    GameTimeInSeconds=@n1.DeltaTimeX
    DebugJointName=neck_02
    Parameters=&/F_Med_StoutWhiz_FaceAcc_Parameters
   )
   > Set Gravity Override(Gravity Override=@n2)
   > n3 = Branch(Condition=Get bIsInFrontEnd())
   > Set Sim Space Settings(Sim Space Settings:Sim Space Settings Structure=Make Sim Space Settings!(
     WorldAlpha:float=0.1
     VelocityScaleZ:float=1
     DampingAlpha:float=1
     MaxLinearVelocity:float=800
     MaxAngularVelocity:float=8
     MaxLinearAcceleration:float=1000
     MaxAngularAcceleration:float=1000
     ExternalLinearDragV:Vector="0,0,0"
     ExternalLinearVelocity:Vector="0,0,0"
     ExternalAngularVelocity:Vector="0,0,0"
    ) -> (SimSpaceSettings:Sim Space Settings Structure) {is=title K2Node_MakeStruct}) -> (Output_Get:Sim Space Settings Structure)
   > Set Gravity Override(Gravity Override="0,0,-500")
   Set Frontend Alpha(execute=@n3.else, Frontend Alpha:double=0) -> (Output_Get:double)
   > Set Sim Space Settings(Sim Space Settings=Break Fort Rigid Body Anim Node Parameters!(FortRigidBodyAnimNodeParameters:"Fort Rigid Body Anim Node Parameters Structure (by ref)"=@n2.OutRigidBodyAnimNodeParameters) -> (ComponentLinearAccScale:Vector=~, ComponentLinearVelScale:Vector=~, SimSpaceSettings:Sim Space Settings Structure) {is=title K2Node_BreakStruct}.SimSpaceSettings)
   Event BlueprintInitializeAnimation()
   > Set Skel_Mesh(Skel_Mesh=GetOwnerSkeletalMeshComponent())
   Event OnBeginPlayingEmote() {is=title K2Node_Event}
   > n4 = Branch(Condition=Get bIsInFrontEnd())
   > Set FrontendHair Alpha(FrontendHair Alpha:double=0) -> (Output_Get:double)
   > Set Frontend Alpha(Frontend Alpha=0.2)
   Set FrontendHair Alpha(execute=@n4.else, FrontendHair Alpha=0)
   > Set Frontend Alpha(Frontend Alpha=0)
   Event OnCharacterCustomizationLoaded()
   n5 = CheckForBackbling()
   > Set Backbling(Backbling:double="To Float (Boolean)"(InBool=@n5)) -> (Output_Get:double)
   > n6 = Branch(Condition=Get bIsInFrontEnd())
   > Set FrontendHair Alpha(FrontendHair Alpha=1)
   > Set Frontend Alpha(Frontend Alpha=0.5)
   > n7 = Branch(Condition=@n5)
   > Set Frontend Backbling(Frontend Backbling:double=1) -> (Output_Get:double)
   Set Frontend Backbling(execute=@n7.else, Frontend Backbling=0)
   Set FrontendHair Alpha(execute=@n6.else, FrontendHair Alpha=0)
   > Set Frontend Alpha(Frontend Alpha=0.5)
   > Set Frontend Backbling(Frontend Backbling=0)
   Event OnEndPlayingEmote() {is=title K2Node_Event}
   > n8 = Branch(Condition=Get bIsInFrontEnd())
   > Set FrontendHair Alpha(FrontendHair Alpha=1)
   > Set Frontend Alpha(Frontend Alpha=0.5)
   Set FrontendHair Alpha(execute=@n8.else, FrontendHair Alpha=0)
   > Set Frontend Alpha(Frontend Alpha=0)
   Event OnBeginDBNOCarried() {is=title K2Node_Event}
   > Set DBNO Carry Alpha(DBNO Carry Alpha:double=1) -> (Output_Get:double)
   Event OnEndDBNOCarried() {is=title K2Node_Event}
   > Set DBNO Carry Alpha(DBNO Carry Alpha=0)
  ]}
]
variables [
 Backbling {type=number}
 Body Mesh {type=SkeletalMeshComponent}
 DBNO Carry Alpha {type=number}
 Frontend Alpha {type=number}
 Frontend Backbling {type=number}
 FrontendHair Alpha {type=number}
 Gravity Override {type=Vector}
 Sim Space Settings {
  type=SimSpaceSettings
  default {
   dampingAlpha=1
   maxAngularAcceleration=10000
   maxAngularVelocity=1000
   maxLinearAcceleration=10000
   maxLinearVelocity=1000
   velocityScaleZ=1
   worldAlpha=0.2
   externalLinearDragV=<0 0 0>
  }
 }
 Skel_Mesh {type=SkeletalMeshComponent}
]
